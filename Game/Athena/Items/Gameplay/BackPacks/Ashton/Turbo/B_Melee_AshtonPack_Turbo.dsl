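; Melee weapon Blueprint for the Ashton Turbo back pack item, layering custom FX and visibility rules onto its parent.
; Pawn hits store the sword tip position, wait briefly, then spawn an impact emitter aimed along the tip's movement.
; Swing Niagara FX starts deactivated, while idle FX follows weapon visibility.
; On equip, the weapon mesh hides if the owning player has the block-throw tag, otherwise any hidden mesh and idle FX are restored.
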(event FortWeapon|EventOnPlayImpactFX
  (params (HitResult :type Hit%20Result%20Structure%20%28by%20ref%29) (ImpactPhysicalSurface :type EPhysicalSurface%20Enum :default "SurfaceType_Default") (SpawnedPSC :type FXSystem%20Component%20Object%20Reference))
  (bind (_bblockinghit _binitialoverlap _time _distance _location _impactpoint _normal _impactnormal _physmat _hitactor) (Collision|BreakHitResult HitResult))
  (bind _swordpos (Variables|Default|GetSwordPos))
  (bind _returnvalue (Transformation|GetWorldLocation _swordpos))
  (|CallParentFunction HitResult ImpactPhysicalSurface SpawnedPSC)
  (Utilities|Casting|CastToFortPawn _hitactor)
  (Variables|Default|SetImpactTipLocation _returnvalue)
  (Utilities|FlowControl|Delay 0.04)
  (Effects|SpawnEmitteratLocation "/Game/Athena/Items/Gameplay/BackPacks/Ashton/Turbo/FX/P_Ashton_Turbo_WeaponHit_Player_Impact.P_Ashton_Turbo_WeaponHit_Player_Impact" _location (Math|Conversions|RotationFromXVector (- _returnvalue (Variables|Default|GetImpactTipLocation)))))

(event EventOnWeaponVisibilityChanged
  (params (bVisible :type Boolean) (bSetForLocalControllerOnly :type Boolean))
  (bind _idlefx_niagara (Variables|Default|GetIdleFX%20Niagara))
  (|CallParentFunction bVisible false)
  (if bVisible
    (Components|Activation|Activate _idlefx_niagara false)
    (else
      (Components|Activation|Deactivate _idlefx_niagara))))

(event EventBeginPlay
  (|CallParentFunction)
  (Components|Activation|Deactivate (Variables|Default|GetSwingFX%20Niagara)))

(event FortWeapon|EventOnEquippedAndAttached
  (bind _weaponmesh (Variables|Visual|GetWeaponMesh))
  (|CallParentFunction)
  (if (GameplayTags|HasMatchingGameplayTag (Utilities|Casting|CastToFortPlayerPawn (Game|GetInstigator)) (Variables|Default|GetBlockThrowTag))
    (Development|SetHiddeninGame _weaponmesh true true)
    (elif (Variables|/Script/Engine.SceneComponent|GetbHiddenInGame _weaponmesh)
      (Development|SetHiddeninGame _weaponmesh false true)
      (Rendering|SetVisibility (Variables|Default|GetIdleFX%20Niagara) true true))))

(fn ConstructionScript
  (|CallParentFunction))
