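; Thrown axe projectile for the Ashton Turbo backpack, steered by laser-targeted homing until it returns to its thrower.
; Timelines ramp flight speed and spin, while a timer triggers replicated lock-on return, ending in a catch event that unhides the held weapon.
; Impacts break buildings under a health cap, and overlaps damage enemies, block their jetpacks or launch players.
; Strike events go to a tracking ability because projectiles cannot send replicated GameplayCues.
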(event Custom|ScaleMovementSpeed
  (bind _projectilemovementcomponent (Variables|Movement|GetProjectileMovementComponent))
  (bind _initalscalingspeed (Variables|BPSetVariables|GetInitalScalingSpeed))
  (bind _slowscalingspeed (Variables|BPSetVariables|GetSlowScalingSpeed))
  ;[1 1 1] Scale the movement speed of the axe after it is thrown
  ;;[1 1 1] Just Get the values once since this is called every frame for a second.
  (Variables|BPSetVariables|SetInitalScalingSpeed (ScalableFloat|GetValueatLevel (Variables|BPSetVariables|GetRow_InitialSpeed)))
  (Variables|BPSetVariables|SetSlowScalingSpeed (ScalableFloat|GetValueatLevel (Variables|BPSetVariables|GetRow_SlowSpeed)))
  (bind (_direction _speedratio) (Timeline|MovementSpeedScale PlayFromStart)
    (:Update
      (if (not (Variables|BPSetVariables|GetIsReturning))
        (Variables|/Script/Engine.ProjectileMovementComponent|SetInitialSpeed (Math|Float|Lerp _initalscalingspeed _slowscalingspeed _speedratio) _projectilemovementcomponent)
        (Variables|/Script/Engine.ProjectileMovementComponent|SetMaxSpeed (Math|Float|Lerp _initalscalingspeed _slowscalingspeed _speedratio) _projectilemovementcomponent)))))

(event EventBeginPlay
  (bind _projectilemovementcomponent (Variables|Movement|GetProjectileMovementComponent))
  (bind _velocity (Variables|/Script/Engine.MovementComponent|GetVelocity _projectilemovementcomponent))
  ;[1 1 1] EVENT -> Begin Play
  (Utilities|FlowControl|Sequence
    (:then_0
      ;[1 1 1] Set Homing variables
      (Projectile|SetHomingStyle 0 0 "LaserTargeted_NoTrace" false :!class /Script/FortniteGame.FortProjectileMovementInterface)
      (Projectile|SetHomingTimeUntilMaxTurnSpeed 0 0 (Variables|HomingValues|GetTimeToMaxTurnSpeed) :!class /Script/FortniteGame.FortProjectileMovementInterface)
      (Projectile|SetHomingTurnSpeedMin 0 0 (Variables|HomingValues|GetHomingTurnRotationMin) :!class /Script/FortniteGame.FortProjectileMovementInterface)
      (Projectile|SetHomingTurnSpeedMax 0 0 (Variables|HomingValues|GetHomingTurnRotationMax) :!class /Script/FortniteGame.FortProjectileMovementInterface)
      (Projectile|SetHomingLaserTargetDistance 0 0 (Variables|HomingValues|GetLaserHomingTargetDistance) :!class /Script/FortniteGame.FortProjectileMovementInterface))
    (:then_1
      (Utilities|FlowControl|Delay 0.01) ; Sometimes Vars don't get set? (Chris M.)
      (Variables|BPSetVariables|SetPreviousVelocity _velocity)
      (Variables|BPSetVariables|SetReturnVelocity (Math|Vector|NegateVector _velocity))
      (CallFunction|ScaleMovementSpeed)
      (bind _asfort_player_pawn (Utilities|Casting|CastToFortPlayerPawn (Game|GetInstigator)))
      (Variables|BPSetVariables|SetPlayerPawn _asfort_player_pawn)
      (Utilities|Time|SetTimerbyEvent (EventDispatchers|CreateEvent :!event Custom|ReturnToOwner) (ScalableFloat|GetValueatLevel (Variables|BPSetVariables|GetRow_AutoReturnTime)))
      (bind _returnvalue (Utilities|Time|SetTimerbyEvent (EventDispatchers|CreateEvent :!event Custom|KillProjectile) 8.0)) ; In case something goes wrong, ensure ability ends
      (Variables|BPSetVariables|SetFailsafeTimer _returnvalue)
      (CallFunction|HideWeapon)
      ;[1 1 1] If player dies, destroy projectile
      (EventDispatchers|BindEventtoOnDied (Variables|BPSetVariables|GetPlayerPawn) (delegate Custom|OnDied))
      (Game|BindEventtoOnDestroyed (EventDispatchers|CreateEvent :!event Custom|ProjectileDestroyed))
      ;[1 1 1] Axe Rotation in air
      (Components|Movement|SetUpdatedComponent (Variables|Default|GetRotatingMovement) (Variables|Default|GetSkelMesh_Weapon))
      (bind (_direction _scalerotation) (Timeline|RotateHammer PlayFromStart)
        (:Update
          (Variables|/Script/Engine.RotatingMovementComponent|SetRotationRate_Roll (Math|Rotator|MakeRotator 0.0 (* _scalerotation (Variables|DefaultTurboProjectile|GetRotationRate))) (Variables|Default|GetRotatingMovement)))) ; Scale up rotation rate so it isn't rotating instantly when it comes out of hand
    )
    (:then_2
      (Utilities|FlowControl|Delay 0.1) ; Turn on Collision a moment later so we have time to pass ref to the projectile tracking ability
      (Collision|SetCollisionEnabled (Variables|Default|GetEnemyZapOverlap) "QueryOnly")
      ;;[1 1 1] Quest Stuff
      (bind _returnvalue_2 (GameplayEffects|ApplyGameplayEffectToSelf :self (Ability|GetAbilitySystemComponent (Variables|BPSetVariables|GetPlayerPawn)) :GameplayEffectClass (Variables|DefaultTurboProjectile|GetGE_QuestCheckTag) :Level 0.0))
      (Variables|BPSetVariables|SetQuestCheckTag_GE _returnvalue_2))))

(event Custom|OnDied
  (params (DamagedActor :type Actor%20Object%20Reference) (Damage :type Float%20%28single-precision%29) (InstigatedBy :type Controller%20Object%20Reference) (DamageCauser :type Actor%20Object%20Reference) (HitLocation :type Vector) (FHitComponent :type Primitive%20Component%20Object%20Reference) (BoneName :type Name) (Momentum :type Vector))
  ;[1 1 1] If player dies, destroy projectile
  (CallFunction|KillProjectile))

(event Custom|CallReturnMovement
  ;[1 1 1] Call Return Movement
  (CallFunction|ResumeMovement-Return))

(event Custom|DestroyBuilding
  ;[1 1 1] Destroy building
  (GameplayEffects|ApplyGameplayEffectToTarget :self (Ability|GetAbilitySystemComponent (Variables|BPSetVariables|GetPlayerPawn)) :GameplayEffectClass "/Game/Athena/Items/Gameplay/BackPacks/Ashton/GE_AshtonPack_Environment_Damage.GE_AshtonPack_Environment_Damage_C" :Target (Ability|GetAbilitySystemComponent (Variables|BPSetVariables|GetLastHitBuilding)) :Level 0.0)
  (Components|Activation|Activate (Variables|Default|GetHitStructureSound) true))

(event Custom|CallMovement
  ;[1 1 1] Call Outgoing Movement
  ;;[0 0 0 0.5] Added this to make sure movement and collision are re-enabled properly
  (Variables|Default|SetResumeMovementRep true)
  (CallFunction|ResumeMovement))

(event OnProjectileStop(ProjectileMovementComponent)
  (params (ImpactResult :type Hit%20Result%20Structure%20%28by%20ref%29))
  (bind (_bblockinghit _binitialoverlap _time _distance _location _impactpoint _normal _impactnormal _physmat _hitactor) (Collision|BreakHitResult ImpactResult))
  ;[1 1 1] EVENT -> On Projectile Stop
  (Variables|BPSetVariables|SetHitLocation _location)
  (Variables|BPSetVariables|SetImpactNormal _normal)
  (Utilities|IsValid _hitactor
    (:Is%20Valid
      (if (Variables|/Script/Engine.Actor|GetbCanBeDamaged _hitactor)
        (bind _asbuilding_actor (Utilities|Casting|CastToBuildingActor _hitactor)
          (:then
            (bind _output_get (Variables|BPSetVariables|SetLastHitBuilding _asbuilding_actor))
            (if (and (> (Health|GetMaxHealth _output_get :!class /Script/FortniteGame.BuildingActor) 0.0) (<= (Health|GetMaxHealth _output_get :!class /Script/FortniteGame.BuildingActor) 5e+03))
              (CallFunction|DestroyBuilding)
              (if (Variables|BPSetVariables|GetReturnToSender)
                (CallFunction|HandleProjectileStop-Return)
                (else
                  (CallFunction|HandleProjectileStop)))
              (else
                _)))
          (:CastFailed
            _))
        (Transformation|SetActorLocation (+ (Transformation|GetActorLocation) (* (Math|Vector|GetUnitDirection(Vector) (Transformation|GetActorLocation) (Transformation|GetActorLocation (Variables|BPSetVariables|GetPlayerPawn))) 1e+02)) false true) ; If we hit something odd then we need to move projectile towards player or it will just register hits constantly
        (CallFunction|HandleProjectileStop-Return)
        (Networking|SwitchHasAuthority
          (:Authority
            (CallFunction|ReturntoOwner)))
        (elif (Variables|BPSetVariables|GetReturnToSender)
          (Collision|IgnoreActorwhenMoving (Variables|Collision|GetCapsuleComponent) _hitactor true)
          (else
            (Components|Activation|Activate (Variables|Default|GetHitStructureSound) false)
            (Variables|BPSetVariables|SetHitIndestructable true)))))))

(event Custom|ProjectileDestroyed
  (params (DestroyedActor :type Actor%20Object%20Reference))
  ;[1 1 1] Quest Stuff - Projectile Destroyed
  (GameplayEffects|RemoveActiveGameplayEffect (Ability|GetAbilitySystemComponent (Variables|BPSetVariables|GetPlayerPawn)) (Variables|BPSetVariables|GetQuestCheckTag_GE) -1))

(event Custom|ReturnToOwner
  ;[1 1 1] Return To Owner
  (bind _distance (Class|MissionBlueprintFunctionLibrary|DistanceBetweenTwoVectors (Transformation|GetActorLocation) (Transformation|GetActorLocation (Variables|BPSetVariables|GetPlayerPawn))))
  (if (>= _distance 75.0)
    (Variables|BPSetVariables|SetIsReturning true)
    (CallFunction|TurnonRecallFX)
    (Variables|BPSetVariables|SetReturnToSender true) ; Set Return Things
    (Networking|ForceNetUpdate)
    (else
      (CallFunction|KillProjectile))))

(event Custom|TurnOnRecallFX
  ;[1 1 1] Turn on return FX and sounds
  (Utilities|FlowControl|DoOnce false
    (:Completed
      (Components|Activation|Activate (Variables|Default|GetRecallParticle) true)
      (if (Pawn|IsLocallyControlled (Variables|BPSetVariables|GetPlayerPawn))
        (Audio|SpawnSoundAttached (Variables|Sounds|GetRecallSound1P) (Variables|Collision|GetCapsuleComponent))
        (else
          (Audio|SpawnSoundAttached (Variables|Sounds|GetRecallSound) (Variables|Collision|GetCapsuleComponent)))))))

(event EventOnTouched
  (params (OtherActor :type Actor%20Object%20Reference) (OtherComp :type Primitive%20Component%20Object%20Reference) (HitResult :type Hit%20Result%20Structure%20%28by%20ref%29) (bIsOverlap :type Boolean))
  (bind (_bblockinghit _binitialoverlap _time _distance _location) (Collision|BreakHitResult HitResult))
  ;[1 1 1] EVENT -> Touched
  (if (not bIsOverlap)
    (Variables|BPSetVariables|SetHitLocation _location)
    (Variables|BPSetVariables|SetPreviousVelocity (Transformation|GetVelocity)) ; Set Previous Velocity so on ResumeMovement we know the direction to go
  ))

(event OnComponentBeginOverlap(EnemyZapOverlap)
  (params (OverlappedComponent :type Primitive%20Component%20Object%20Reference) (OtherActor :type Actor%20Object%20Reference) (OtherComp :type Primitive%20Component%20Object%20Reference) (OtherBodyIndex :type Integer) (bFromSweep :type Boolean) (SweepResult :type Hit%20Result%20Structure%20%28by%20ref%29))
  (bind _returnvalue (Game|GetInstigator))
  (bind _returnvalue_2 (Ability|GetAbilitySystemComponent _returnvalue))
  ;[1 1 1] EVENT -> On Overlap - Player Zap
  (if (not (or (Utilities|Equal(Object) (Variables|Actor|GetInstigator) OtherActor) (Team|OnSameTeam OtherActor (Variables|BPSetVariables|GetPlayerPawn))))
    (bind _asfort_player_pawn (Utilities|Casting|CastToFortPlayerPawn OtherActor))
    (GameplayEffects|ApplyGameplayEffectToTarget :self _returnvalue_2 :GameplayEffectClass (Variables|DefaultTurboProjectile|GetGE_Damage) :Target (Ability|GetAbilitySystemComponent _asfort_player_pawn) :Level 0.0)
    (GameplayEffects|ApplyGameplayEffectToTarget :self _returnvalue_2 :GameplayEffectClass (Variables|DefaultTurboProjectile|GetGE_BlockJetPack) :Target (Ability|GetAbilitySystemComponent _asfort_player_pawn) :Level 0.0) ; Block their jetpack use
    (Ability|SendGameplayEventtoActor (Game|GetInstigator) "(TagName=\"Event.Athena.Item.Ashton.Turbo.Strike\")" (Utilities|Struct|MakeGameplayEventData :Instigator self :Target _asfort_player_pawn :ContextHandle "()" :InstigatorTags "(GameplayTags=)" :TargetTags "(GameplayTags=)" :EventMagnitude 0.0 :TargetData "()")) ; Send event to the tracking GA to play the GC at the location with references to the Axe and the hit target - Have to do it this way since we cannot send replicated GCs on Projectiles
  ))

(event Custom|HideWeapon
  ;[1 1 1] Hide Weapon
  (bind _asb_melee_ashton_pack_turbo (Utilities|Casting|CastToB_Melee_AshtonPack_Turbo (Variables|/Script/FortniteGame.FortPawn|GetCurrentWeapon (Variables|BPSetVariables|GetPlayerPawn))))
  (Development|SetHiddeninGame (Variables|/Script/FortniteGame.FortWeapon|GetWeaponMesh _asb_melee_ashton_pack_turbo) true true)
  (Rendering|SetVisibility (Variables|/Script/FortniteGame.FortWeapon|GetWeaponMesh _asb_melee_ashton_pack_turbo) false true))

(event ProjectileMovement|EventOnReachedLockTargetDistanceThreshold
  ;[1 1 1] EVENT -> Reached Homing Target - Axe came in range of owner.
  (if (Variables|BPSetVariables|GetIsReturning)
    (bind _outhits (Collision|MultiLineTraceForObjects (Transformation|GetActorLocation) (Transformation|GetActorLocation (Variables|BPSetVariables|GetPlayerPawn)) (Variables|BPSetVariables|GetObjectTypes))) ; When Hammer is close, check for anything between hammer and us
    (for _array_element _outhits
      (if (Variables|/Script/Engine.Actor|GetbCanBeDamaged (.HitActor (Collision|BreakHitResult _array_element)))
        (bind _asbuilding_actor (Utilities|Casting|CastToBuildingActor (.HitActor (Collision|BreakHitResult _array_element))))
        (Variables|BPSetVariables|SetLastHitBuilding _asbuilding_actor)
        (CallFunction|DestroyBuilding)))
    (CallFunction|KillProjectile)))

(event ProjectileMovement|EventOnHomingPassedTarget
  ;[1 1 1] EVENT -> Reached Homing Target - Axe came in range of owner.
  (Projectile|SetHomingTurnSpeedMax 0 0 5e+04 :!class /Script/FortniteGame.FortProjectileMovementInterface))

(event Custom|UnhideMesh
  ;[1 1 1] Unhide mesh.  Needs to be multicast or sometimes the destroy event wont reach other clients
  (bind _asb_melee_ashton_pack_turbo (Utilities|Casting|CastToB_Melee_AshtonPack_Turbo (Variables|/Script/FortniteGame.FortPawn|GetCurrentWeapon (Variables|BPSetVariables|GetPlayerPawn))))
  (Development|SetHiddeninGame (Variables|/Script/FortniteGame.FortWeapon|GetWeaponMesh _asb_melee_ashton_pack_turbo) false true)
  (Rendering|SetVisibility (Variables|/Script/FortniteGame.FortWeapon|GetWeaponMesh _asb_melee_ashton_pack_turbo) true true))

(event Custom|KillProjectile
  (bind _playerpawn (Variables|BPSetVariables|GetPlayerPawn))
  ;[1 1 1] Kill Projectile
  (GameplayEffects|RemoveActiveEffectswithGrantedTags (Ability|GetAbilitySystemComponent _playerpawn) (Variables|DefaultTurboProjectile|GetBlockThrowTag))
  (Ability|SendGameplayEventtoActor _playerpawn (Variables|BPSetVariables|GetEvent_Catch))
  (CallFunction|UnhideMesh)
  (Actor|DestroyActor)
  (Components|DestroyComponent (Variables|Default|GetInAirSound)))

(event Custom|PlayerLaunch
  (params (PlayerPawn :type Fort%20Player%20Pawn%20Object%20Reference))
  ;[1 1 1] Launch Player
  (Character|LaunchCharacter PlayerPawn (Math|Vector|MakeVector (.x (* (Math|Vector|Normalize (Math|Vector|GetUnitDirection(Vector) (Transformation|GetActorLocation) (Transformation|GetActorLocation PlayerPawn))) (ScalableFloat|GetValueatLevel (Variables|DefaultTurboProjectile|GetPlayerLaunchVelocity)))) (.y (* (Math|Vector|Normalize (Math|Vector|GetUnitDirection(Vector) (Transformation|GetActorLocation) (Transformation|GetActorLocation PlayerPawn))) (ScalableFloat|GetValueatLevel (Variables|DefaultTurboProjectile|GetPlayerLaunchVelocity)))) (+ (.z (* (Math|Vector|Normalize (Math|Vector|GetUnitDirection(Vector) (Transformation|GetActorLocation) (Transformation|GetActorLocation PlayerPawn))) (ScalableFloat|GetValueatLevel (Variables|DefaultTurboProjectile|GetPlayerLaunchVelocity)))) (ScalableFloat|GetValueatLevel (Variables|DefaultTurboProjectile|GetPlayerLaunchVelocity)))) true true))

(event OnComponentBeginOverlap(LaunchOverlap)
  (params (OverlappedComponent :type Primitive%20Component%20Object%20Reference) (OtherActor :type Actor%20Object%20Reference) (OtherComp :type Primitive%20Component%20Object%20Reference) (OtherBodyIndex :type Integer) (bFromSweep :type Boolean) (SweepResult :type Hit%20Result%20Structure%20%28by%20ref%29))
  ;[1 1 1] EVENT -> On Overlap - Launch
  (bind _playertolaunch (Utilities|ShouldLaunchPlayer OtherActor)
    (:Yes
      (CallFunction|PlayerLaunch _playertolaunch))))

(fn CalculateNewVelocity
  (params (TraceStart :type Vector) (TraceEnd :type Vector))
  (return (* (Math|Vector|Normalize (- TraceEnd TraceStart)) (* (Math|Vector|VectorLength (Variables|BPSetVariables|GetPreviousVelocity)) 0.9)) (>= (* (Math|Vector|VectorLength (Variables|BPSetVariables|GetPreviousVelocity)) 0.9) 2e+03)))

(fn HandleProjectileStop
  (Networking|SwitchHasAuthority
    (:Authority
      (Utilities|Time|SetTimerbyEvent (EventDispatchers|CreateEvent :!event CallMovement) 1e-05))))

(fn HandleProjectileStop-Return
  (Networking|SwitchHasAuthority
    (:Authority
      (Utilities|Time|SetTimerbyEvent (EventDispatchers|CreateEvent :!event CallReturnMovement) 1e-05))))

(fn LaunchPawn
  (params (LaunchedPawn :type Fort%20Player%20Pawn%20Object%20Reference))
  (Character|LaunchCharacter LaunchedPawn (Math|Vector|MakeVector (.x (Transformation|GetVelocity)) (.y (Transformation|GetVelocity)) (+ (.z (Transformation|GetVelocity)) (Variables|DefaultTurboProjectile|GetPlayerLaunchStrength))) true true))

(fn OnRep_ResumeMovementRep
  (Collision|SetCollisionEnabled (Variables|Collision|GetCapsuleComponent) "NoCollision")
  ;[1 1 1] Set Homing variables
  (Projectile|SetHomingStyle 0 0 "LaserTargeted_NoTrace" false :!class /Script/FortniteGame.FortProjectileMovementInterface)
  (Projectile|SetHomingTimeUntilMaxTurnSpeed 0 0 (Variables|HomingValues|GetTimeToMaxTurnSpeed) :!class /Script/FortniteGame.FortProjectileMovementInterface)
  (Projectile|SetHomingTurnSpeedMin 0 0 (Variables|HomingValues|GetHomingTurnRotationMin) :!class /Script/FortniteGame.FortProjectileMovementInterface)
  (Projectile|SetHomingTurnSpeedMax 0 0 (Variables|HomingValues|GetHomingTurnRotationMax) :!class /Script/FortniteGame.FortProjectileMovementInterface)
  (Projectile|SetHomingLaserTargetDistance 0 0 (Variables|HomingValues|GetLaserHomingTargetDistance) :!class /Script/FortniteGame.FortProjectileMovementInterface)
  (Collision|SetCollisionEnabled (Variables|Collision|GetCapsuleComponent) "QueryOnly"))

(fn OnRep_ReturnToSender
  (bind _row_returnspeed (Variables|BPSetVariables|GetRow_ReturnSpeed))
  (bind _returnvalue (ScalableFloat|GetValueatLevel _row_returnspeed))
  (bind _self self)
  (bind _projectilemovementcomponent (Variables|Movement|GetProjectileMovementComponent))
  (Collision|SetCollisionEnabled (Variables|Collision|GetCapsuleComponent) "NoCollision")
  (Projectile|SetHomingStyle 0 _self "LockOn" false :!class /Script/FortniteGame.FortProjectileMovementInterface)
  (Projectile|SetHomingTarget 0 _self (Variables|BPSetVariables|GetPlayerPawn) 0 :!class /Script/FortniteGame.FortProjectileMovementInterface)
  (Projectile|SetHomingLockTargetDistanceThreshold 0 _self 2e+02 :!class /Script/FortniteGame.FortProjectileMovementInterface)
  (Projectile|SetHomingTargetOffset _self (Variables|HomingValues|GetHomingTargetOffset) :!class /Script/FortniteGame.FortProjectileMovementInterface)
  (Variables|/Script/Engine.ProjectileMovementComponent|SetInitialSpeed _returnvalue _projectilemovementcomponent)
  (Variables|/Script/Engine.ProjectileMovementComponent|SetMaxSpeed _returnvalue _projectilemovementcomponent)
  (Projectile|SetHomingTurnSpeedMax 0 self (Variables|HomingValues|GetTurnSpeedMax_Return) :!class /Script/FortniteGame.FortProjectileMovementInterface)
  (Collision|SetCollisionEnabled (Variables|Collision|GetCapsuleComponent) "QueryOnly")
  (CallFunction|ResumeMovement-Return))

(fn Resume_Movement
  (Components|Movement|SetUpdatedComponent (Variables|Movement|GetProjectileMovementComponent) (Variables|Collision|GetCapsuleComponent))
  (Variables|/Script/Engine.MovementComponent|SetVelocity (Variables|BPSetVariables|GetPreviousVelocity) (Variables|Movement|GetProjectileMovementComponent)))

(fn ResumeMovement-Return
  (Components|Movement|SetUpdatedComponent (Variables|Movement|GetProjectileMovementComponent) (Variables|Collision|GetCapsuleComponent)))

(fn ConstructionScript
  (bind _enemyzapoverlap (Variables|Default|GetEnemyZapOverlap))
  (Collision|SetCollisionEnabled _enemyzapoverlap "NoCollision")
  (Components|Sphere|SetSphereRadius _enemyzapoverlap (ScalableFloat|GetValueatLevel (Variables|DefaultTurboProjectile|GetPlayerZapRadius)) true)
  (|CallParentFunction))
