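# Unlit translucent deferred decal material for the Ashton Turbo backpack FX, drawing a charred scorch mark with glowing blue cracks.
# A glass-crack texture in radial UVs is masked by a sphere mask and ground char, then tinted between edge and center colors.
# Decal lifetime opacity drives an offset fade that thins and dims the cracks early through a changing power exponent.
# The char texture's opacity uses a multiplied, clamped lifetime fade, so it outlasts the glow.
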Material /Game/Athena/Items/Gameplay/BackPacks/Ashton/Turbo/FX/MD_Ashton_Turbo_Decal_ScorchCracks
graph {
 properties {
  bAutomaticallySetUsageInEditor=false
  blendMode=BLEND_Translucent
  materialDomain=MD_DeferredDecal
  pixelDepthOffsetMode=PDOM_Legacy
  shadingModel=MSM_Unlit
  usedShadingModels=Unlit
 }
 nodes [
  n1 = DecalLifetimeOpacity()
  n2 = Clamp(None=LinearInterpolate(A=ScalarParameter() {defaultValue=-3, parameterName=EmissiveFadeOffset}, Alpha=Clamp(None=@n1.Opacity)) {constA=-7})
  n3 = MaterialFunctionCall(Vector or UVs=Add(A=TextureCoordinate() {uTiling=1.5, vTiling=1.5}, B=Constant2Vector() {g=-0.75, r=-0.75})) {materialFunction=&VectorToRadialValue}
  n4 = AppendVector(A=ScalarParameter() {defaultValue=1, parameterName=CrackTilingU}, B=ScalarParameter() {defaultValue=1, parameterName=CrackTilingV})
  n5 = VectorParameter() {parameterName=ElecMask, defaultValue=<rgba 0.5 0.5 0 1>}
  n6 = Multiply(
   A=TextureSample(UVs=Multiply(A=@n3.Radial Coordinates, B=@n4)) {texture=/Game/Effects/Fort_Effects/Textures/Tile/T_Glass_Cracks_M}.G
   B=SphereMask(
    A=TextureCoordinate()
    B=ComponentMask(None=@n5.RGB) {g=true, r=true}
    Radius=ScalarParameter() {defaultValue=0.4, parameterName=ElecMask_Radius}
    Hardness=ScalarParameter() {defaultValue=0.25, parameterName=ElecMask_Hardness}
   )
  )
  n7 = TextureSample() {samplerType=SAMPLERTYPE_LinearColor, texture=/Game/Effects/Fort_Effects/Gameplay/SmokeDustExplosions/Textures/T_GroundChar_01}
  n8 = Multiply(A=Power(Base=@n2), B=Power(Base=Multiply(A=@n6, B=@n7.G), Exp=LinearInterpolate(Alpha=@n2) {constA=14}))
  n9 = Multiply(
   A=LinearInterpolate(
    A=VectorParameter() {group=Emissive, parameterName=EmissiveEdge, defaultValue=<rgba 0 0.3598 2 1>}.RGB
    B=VectorParameter() {group=Emissive, parameterName=EmissiveCenter, defaultValue=<rgba 1.1719 2.1658 5 1>}.RGB
    Alpha=@n8
   )
   B=@n8
  )
  MaterialFunctionCall(LDR Input=Clamp(None=@n9), HDR Tint=@n9) {materialFunction=&DeriveHDRfromLDR}
  n10 = TextureSampleParameter2D() {
   parameterName=DecalTexture
   samplerType=SAMPLERTYPE_LinearColor
   texture=&T_GroundChar_01
  }
  n11 = ScalarParameter() {defaultValue=6, group=Timestamp, parameterName=Fade Transition Multiplier}
  Multiply(A=@n10.R, B=Clamp(None=Multiply(A=@n1.Opacity, B=@n11) {constB=4}))
  ScalarParameter() {parameterName=Metallic}
  ScalarParameter() {parameterName=Specular}
  ScalarParameter() {defaultValue=1, parameterName=Roughness}
  VectorParameter() {parameterName=DecalColor, defaultValue=<rgba 0 0 0 1>}
 ]
}
