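# Translucent unlit fire flipbook material for sprite particles in the Ashton Turbo back bling effects.
# Particle color tints the squared red channel of an 8x8 SubUV fire texture, passed through an emissive adjustment function.
# Opacity combines particle alpha with sharpened flipbook alpha in a depth fade, whose distance varies with panning tile noise.
# Dynamic parameters X and Scale offset and tile that noise; LerpNoise and BlackBodyTemp are unused.
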Material /Game/Athena/Items/Gameplay/BackPacks/Ashton/Turbo/FX/M_Ashton_Turbo_FireSubUV_Hollow
graph {
 properties {
  bUsedWithNiagaraSprites=true
  bUsedWithParticleSprites=true
  blendMode=BLEND_Translucent
  pixelDepthOffsetMode=PDOM_Legacy
  shadingModel=MSM_Unlit
  usedShadingModels=Unlit
 }
 nodes [
  n1 = ParticleColor()
  n2 = TextureCoordinate()
  n3 = TextureSampleParameterSubUV!(UVs=@n2, Apply View MipBias) {parameterName=FireSubUV, texture=/Game/Effects/Fort_Effects/Textures/Fire/T_FireFluid_1_8X8}
  n4 = TextureCoordinate() {vTiling=0.5}
  n5 = DynamicParameter() {defaultValue=<rgba 0 0 0 1>, paramNames [LerpNoise, X, BlackBodyTemp, Scale]}
  n6 = AppendVector(A=@n5.X, B=Add(A=@n5.X) {constB=2})
  n7 = Panner(Coordinate=Multiply(A=Add(A=@n4, B=@n6), B=Floor(None=@n5.Scale)) {desc=Scale}) {speedY=-0.125}
  DepthFade(
   Opacity=Multiply(A=@n1.A, B=Power(Base=@n3.A) {constExponent=1.2})
   FadeDistance=LinearInterpolate(Alpha=TextureSample(UVs=@n7) {texture=/Game/Effects/Fort_Effects/Textures/Fire/T_FireTile}.A) {constA=60, constB=120, desc=Bias Noise}
  )
  n8 = TextureSampleParameterSubUV(UVs=@n2) {parameterName=FireSubUV, texture=&T_FireFluid_1_8X8}
  MaterialFunctionCall(Emissive=Multiply(A=@n1.RGB, B=Multiply(A=@n8.R, B=@n8.R))) {materialFunction=&MF_FX_EmissiveAdjustments}
 ]
}
