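# Translucent, unlit, two-sided particle material for the ground-impact shockwave effect of the Ashton Turbo backpack.
# Opacity multiplies Fresnel times one-minus-Fresnel bands, particle alpha, panning fire noise and two flow-map-distorted electric noise layers, softened by depth fade.
# A dynamic parameter drives the second band's exponent and noise pan speed.
# Emissive is particle color, hue-shifted by Fresnel and passed through the shared FX emissive-adjustment function.
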Material /Game/Athena/Items/Gameplay/BackPacks/Ashton/Turbo/FX/M_Ashton_Turbo_GroundImpact_Shockwave
graph {
 properties {
  bUsedWithMeshParticles=true
  bUsedWithParticleSprites=true
  blendMode=BLEND_Translucent
  pixelDepthOffsetMode=PDOM_Legacy
  shadingModel=MSM_Unlit
  twoSided=true
  usedShadingModels=Unlit
 }
 nodes [
  n1 = Fresnel(ExponentIn=Constant() {r=0.8}) {baseReflectFraction=0.001}
  n2 = Multiply(A=Multiply(A=OneMinus(None=@n1), B=@n1), B=ParticleColor().A) {constA=1}
  n3 = DynamicParameter() {defaultValue=<rgba 3 1 1 1>, paramNames [MiddleFade,EdgeFade,PanSpeed,Ref]}
  n4 = Fresnel(ExponentIn=@n3.MiddleFade) {baseReflectFraction=0.001}
  n5 = Multiply(A=OneMinus(None=@n4), B=@n4)
  n6 = Constant() {r=1}
  n7 = TextureCoordinate() {uTiling=4, vTiling=2}
  n8 = Multiply(A=Time(), B=@n3.PanSpeed)
  n9 = TextureCoordinate() {vTiling=3}
  n10 = Multiply(
   A=TextureSample(UVs=Panner(Coordinate=@n7, Time=@n8) {speedY=-1}) {texture=/Game/Global/Textures/Default/Noise/T_FireNoise_02}.G
   B=TextureSample(UVs=Panner(Coordinate=@n9, Time=@n8) {speedY=-1}) {texture=&T_FireNoise_02}.G
  )
  n11 = Multiply(A=TextureCoordinate() {vTiling=2}, B=ScalarParameter() {defaultValue=1, parameterName=MacroTile})
  n12 = Time()
  n13 = ScalarParameter() {defaultValue=5, parameterName=Macro UV Distortion Panning Time}
  n14 = TextureSample(UVs=Panner(Coordinate=@n11, Time=Multiply(A=@n12, B=@n13)) {speedX=0.012, speedY=0.02}) {samplerType=SAMPLERTYPE_LinearColor, texture=/Game/Effects/Fort_Effects/Creatures/Elemental/Fire/MergedFireFlowMap}
  n15 = Multiply(A=Subtract(A=ComponentMask(None=@n14.RGB) {g=true, r=true}) {constB=0.5})
  n16 = ScalarParameter() {defaultValue=0.25, parameterName=Macro UV Tile Distortion Amount}
  n17 = Add(A=Power(Base=Multiply(A=@n15, B=@n16) {constB=0.3}, Exp=ScalarParameter() {defaultValue=1}), B=@n11)
  n18 = TextureSample(UVs=Panner(Coordinate=@n17, Time=@n8) {speedY=-0.1}) {samplerType=SAMPLERTYPE_LinearColor, texture=/Game/Characters/Player/Male/Medium/Bodies/M_MED_SRDrift/Textures/T_ElectricNoise}
  n19 = ScalarParameter() {defaultValue=0.75, parameterName=Second itation of glow scale}
  n20 = TextureSample(UVs=Panner(Coordinate=Multiply(A=@n17, B=@n19) {constB=2}, Time=@n8) {speedY=-0.15}) {samplerType=SAMPLERTYPE_LinearColor, texture=&T_ElectricNoise}
  n21 = Power(
   Base=Multiply(A=Multiply(A=@n18.G, B=@n20.G), B=ScalarParameter() {defaultValue=10, parameterName=ElectricaIntensity})
   Exp=ScalarParameter() {defaultValue=3, parameterName=ElectricalPOwer}
  )
  n22 = Constant() {r=4}
  n23 = Multiply(A=Power(Base=Multiply(A=@n5, B=@n6)) {constExponent=1}, B=Multiply(A=Power(Base=@n10) {constExponent=4}, B=Multiply(A=@n21, B=@n22)))
  DepthFade(Opacity=Multiply(A=Power(Base=@n2), B=Clamp(None=@n23)))
  n24 = ScalarParameter() {defaultValue=0.025, parameterName=HueShiftIntensity}
  MaterialFunctionCall(Emissive=MaterialFunctionCall(Hue Shift Percentage=Multiply(A=@n1, B=@n24), Texture=ParticleColor().RGB) {materialFunction=&HueShift}.Result) {materialFunction=&MF_FX_EmissiveAdjustments}
 ]
}
