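# Translucent unlit sprite-particle material for lightning arcs on the Ashton Turbo backpack.
# Dynamic parameters drive V tiling, erosion, panner speed and a bend that bows the UVs along V, while floored, sine-wobbled particle time steps the muzzle-ring texture pan for jumpy motion.
# Opacity combines the eroded texture minus an edge mask, particle alpha, optional vertex-color masking and layered time-sine flicker.
# Particle color feeds emissive through MF_FX_EmissiveAdjustments.
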Material /Game/Athena/Items/Gameplay/BackPacks/Ashton/Turbo/FX/M_Ashton_Turbo_LightningArcs
graph {
 properties {
  bEnableResponsiveAA=true
  bUsedWithNiagaraSprites=true
  bUsedWithParticleSprites=true
  blendMode=BLEND_Translucent
  pixelDepthOffsetMode=PDOM_Legacy
  shadingModel=MSM_Unlit
  usedShadingModels=Unlit
 }
 nodes [
  n1 = ParticleColor()
  n2 = DynamicParameter() {defaultValue=<rgba 1 0 1 1>, paramNames [Panner Speed, Erode, V Tile, Bend]}
  n3 = ComponentMask(None=TextureCoordinate()) {g=true}
  n4 = Power(Base=Multiply(A=OneMinus(None=@n3), B=@n3)) {constExponent=3}
  n5 = Add(
   A=AppendVector(A=Multiply(A=@n2.Bend, B=@n4) {constA=25}, B=Constant())
   B=Add(A=TextureCoordinate() {uTiling=2}, B=Constant2Vector() {r=-0.5})
  )
  n6 = AppendVector(
   A=ScalarParameter() {defaultValue=1, parameterName=U Tile}
   B=Multiply(
    A=DynamicParameter() {defaultValue=<rgba 1 0 1 1>, paramNames [Panner Speed, Erode, V Tile, Bend]}.V Tile
    B=ScalarParameter() {defaultValue=1, parameterName=V Tile}
   ) {constA=1}
  )
  n7 = ParticleRelativeTime()
  n8 = StaticSwitchParameter(
   True=Multiply(A=DynamicParameter() {defaultValue=<rgba 1 0 1 1>, paramNames [Panner Speed, Erode, V Tile, Bend]}.Panner Speed, B=@n7)
   False=@n7
  ) {parameterName=Use PRT Multiply?}
  n9 = Sine(None=@n8) {period=0.77}
  n10 = Sine(None=@n8) {period=1.12}
  n11 = DynamicParameter() {defaultValue=<rgba 1 0 1 1>, paramNames [Panner Speed, Erode, V Tile, Bend]}
  n12 = Add(A=Abs(None=Add(A=@n9, B=@n10)), B=Add(A=Multiply(A=@n8, B=@n11.Panner Speed), B=@n11.Panner Speed))
  n13 = TextureSample!(UVs=Panner(Coordinate=Multiply(A=@n5, B=@n6), Time=Floor(None=@n12)) {bFractionalPart=true, speedY=1.25}, Tex, Level, Apply View MipBias) {mipValueMode=TMVM_MipLevel, samplerType=SAMPLERTYPE_LinearColor, texture=/Game/Effects/Fort_Effects/Textures/Tile/T_AMR_MuzzleRing_01}
  n14 = Multiply(A=ParticleRelativeTime(), B=ScalarParameter() {defaultValue=2, parameterName=PRT Scale})
  n15 = Multiply(
   A=LinearInterpolate(A=@n13.B, B=@n13.B, Alpha=Clamp(None=@n14))
   B=Add(A=@n13.G, B=DynamicParameter() {defaultValue=<rgba 1 0 1 1>, paramNames [Panner Speed, Erode, V Tile, Bend]}.Erode)
  )
  n16 = ComponentMask(None=TextureCoordinate()) {g=true}
  n17 = Add(
   A=TextureSample!(UVs, Tex, Apply View MipBias) {samplerType=SAMPLERTYPE_LinearColor, texture=&T_AMR_MuzzleRing_01}.A
   B=Power(Base=OneMinus(None=@n16)) {constExponent=4}
  )
  n18 = Multiply(A=@n1.A, B=Clamp(None=Subtract(A=@n15, B=@n17)))
  n19 = VertexColor()
  n20 = Multiply(
   A=@n18
   B=StaticComponentMaskParameter(None=AppendVector(A=@n19, B=@n19)) {defaultA=true, defaultB=true, defaultG=true, defaultR=true, parameterName=Vertex Color Component Mask}
  )
  n21 = LinearInterpolate(Alpha=ScalarParameter() {defaultValue=1, parameterName=RandomOffsetPerClient_Mesh})
  n22 = Sine(None=Add(A=Time() {bOverride_Period=true, period=2.6}, B=@n21)) {period=2.6}
  n23 = Sine(None=Add(A=Time() {bOverride_Period=true, period=1.57}, B=@n21)) {period=1.57}
  n24 = Time() {bOverride_Period=true, period=1.24}
  n25 = LinearInterpolate(A=Constant() {r=0.15}, B=Constant() {r=0.85}, Alpha=Sine(None=Add(A=@n24, B=@n21)) {period=1.24}) {constA=0.15, constB=0.85}
  Clamp(None=Multiply(A=StaticSwitchParameter(True=@n20, False=@n18) {parameterName=Use Vertex Color Mask?}, B=LinearInterpolate(A=@n22, B=@n23, Alpha=@n25)))
  MaterialFunctionCall(Emissive=@n1.RGB) {materialFunction=&MF_FX_EmissiveAdjustments}
 ]
}
