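# Translucent unlit mesh-particle material for the lightning sphere effect on the Ashton Turbo backpack.
# World position offset combines SplineThicken line width, a directional offset and time-driven sine wobble, modulated per particle through dynamic parameters.
# Flow-map-distorted, panning electric noise is multiplied and sharpened, then masked by a beam flare, UV edge falloff, vertex color and particle alpha.
# Particle color feeds emissive adjustments.
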Material /Game/Athena/Items/Gameplay/BackPacks/Ashton/Turbo/FX/M_Ashton_Turbo_LightningSphere
graph {
 properties {
  bUsedWithMeshParticles=true
  blendMode=BLEND_Translucent
  pixelDepthOffsetMode=PDOM_Legacy
  shadingModel=MSM_Unlit
  twoSided=true
  usedShadingModels=Unlit
 }
 nodes [
  n1 = DynamicParameter() {defaultValue=<rgba 1 1 -1 1>, paramNames [Random,WobbleMul,OffsetMesh,LineWidth]}
  n2 = Multiply(A=ScalarParameter() {defaultValue=5, parameterName="WidthBase (S)"}, B=@n1.LineWidth)
  n3 = Multiply(
   A=@n1.LineWidth
   B=LinearInterpolate(
    A=ScalarParameter() {defaultValue=50, parameterName=WidthTip_LOW}
    B=ScalarParameter() {defaultValue=100, parameterName=WidthTip_HIGH}
    Alpha=VertexColor().G
   )
  )
  n4 = Transform(None=VectorParameter() {parameterName=OffsetDistance, defaultValue=<rgba 0 1 0 0>}.RGB)
  n5 = LinearInterpolate(Alpha=DynamicParameter() {defaultValue=<rgba 1 1 -1 1>, paramNames [Random,WobbleMul,OffsetMesh,LineWidth]}.OffsetMesh) {constB=150}
  n6 = Constant()
  n7 = TextureCoordinate()
  n8 = DynamicParameter() {defaultValue=<rgba 1 1 -1 1>, paramNames [Random,WobbleMul,OffsetMesh,LineWidth]}
  n9 = Multiply(A=Add(A=@n8.Random, B=Time()), B=ScalarParameter() {defaultValue=3, parameterName=WobbleSpeed})
  n10 = ComponentMask(None=Add(A=@n7, B=Multiply(A=@n9) {constB=-0.6})) {g=true}
  n11 = ComponentMask(None=Add(A=@n7, B=Multiply(A=@n9) {constB=-1})) {g=true}
  n12 = Add(A=Multiply(A=Sine(None=@n10) {period=0.2}) {constB=3}, B=Sine(None=@n11) {period=0.8})
  n13 = Multiply(A=Sine(None=@n10) {period=0.23}) {constB=3}
  n14 = Sine(None=@n11) {period=0.6}
  n15 = Transform(None=AppendVector(A=AppendVector(A=@n6, B=@n12), B=Add(A=@n13, B=@n14))) {transformSourceType=TRANSFORMSOURCE_Local}
  n16 = ScalarParameter() {defaultValue=5, parameterName=Wobble}
  Add(
   A=Add(
    A=MaterialFunctionCall(WidthBase=@n2, WidthTip=@n3) {materialFunction=&SplineThicken}.WorldPositionOffset
    B=Multiply(A=@n4, B=@n5)
   )
   B=Multiply(A=Normalize(VectorInput=@n15), B=Multiply(A=@n16, B=@n8.WobbleMul))
  )
  Add()
  n17 = ParticleColor()
  MaterialFunctionCall(Emissive=@n17.RGB) {materialFunction=&MF_FX_EmissiveAdjustments}
  n18 = ComponentMask(None=TextureCoordinate()) {g=true}
  n19 = ComponentMask(None=TextureCoordinate()) {r=true}
  n20 = Multiply(A=Multiply(A=@n18, B=OneMinus(None=@n18)), B=Multiply(A=@n19, B=OneMinus(None=@n19)))
  n21 = VertexColor()
  n22 = Multiply(
   A=Add(A=AppendVector(A=@n21.R, B=@n21.R), B=TextureCoordinate() {uTiling=0.1, vTiling=0.5})
   B=ScalarParameter() {defaultValue=1, parameterName=MacroTile}
  )
  n23 = Multiply(A=Time(), B=ScalarParameter() {defaultValue=6, parameterName=Macro UV Distortion Panning Time})
  n24 = ComponentMask(None=TextureSample(UVs=Panner(Coordinate=@n22, Time=@n23) {speedY=0.02}) {samplerType=SAMPLERTYPE_LinearColor, texture=/Game/Effects/Fort_Effects/Creatures/Elemental/Fire/MergedFireFlowMap}.RGB) {g=true, r=true}
  n25 = Multiply(A=Multiply(A=Subtract(A=@n24) {constB=0.5}), B=ScalarParameter() {defaultValue=0.5, parameterName=Macro UV Tile Distortion Amount}) {constB=0.3}
  n26 = Add(A=@n25, B=@n22)
  n27 = Multiply(A=@n26, B=ScalarParameter() {defaultValue=0.5, parameterName=Second itation of glow scale}) {constB=2}
  n28 = Multiply(
   A=TextureSample(UVs=Panner(Coordinate=@n26) {speedY=-0.2}) {samplerType=SAMPLERTYPE_LinearColor, texture=/Game/Characters/Player/Male/Medium/Bodies/M_MED_SRDrift/Textures/T_ElectricNoise}.G
   B=TextureSample(UVs=Panner(Coordinate=@n27) {speedY=-0.25}) {samplerType=SAMPLERTYPE_LinearColor, texture=&T_ElectricNoise}.G
  )
  n29 = ScalarParameter() {defaultValue=25, parameterName=ElectricaIntensity}
  n30 = Add(A=AppendVector(A=Constant(), B=VertexColor().G), B=Panner(Coordinate=TextureCoordinate()) {speedY=-0.25})
  n31 = Multiply(
   A=Power(Base=Multiply(A=@n28, B=@n29), Exp=ScalarParameter() {defaultValue=2, parameterName=ElectricalPOwer})
   B=TextureSample(UVs=Add(A=@n25, B=@n30)) {texture=/Game/Effects/Fort_Effects/Textures/Flares/T_Beam_Flare_Core}.G
  )
  Multiply(A=VertexColor().B, B=Multiply(A=@n17.A, B=Multiply(A=@n20, B=@n31)))
  ParticleColor()
 ]
}
