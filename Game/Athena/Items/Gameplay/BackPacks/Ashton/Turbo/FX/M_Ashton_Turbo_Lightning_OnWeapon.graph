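# Unlit translucent material for the Ashton Turbo backpack's lightning-on-weapon effect, built for mesh particles and skeletal meshes.
# Layered sine flicker and a travelling vertex-color-driven sphere mask modulate hue shift, emissive color and the opacity of panning beam textures.
# World position offset combines SplineThicken, driven by a dynamic Thickness parameter, with time-stepped noise jitter and curl-style vector noise.
# Mobile, forward and low-quality switches bypass depth fade and vector noise.
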Material /Game/Athena/Items/Gameplay/BackPacks/Ashton/Turbo/FX/M_Ashton_Turbo_Lightning_OnWeapon
graph {
 properties {
  bUsedWithMeshDeformer=true
  bUsedWithMeshParticles=true
  bUsedWithNiagaraMeshParticles=true
  bUsedWithSkeletalMesh=true
  blendMode=BLEND_Translucent
  pixelDepthOffsetMode=PDOM_Legacy
  shadingModel=MSM_Unlit
  twoSided=true
  usedShadingModels=Unlit
 }
 nodes [
  n1 = ScalarParameter() {parameterName=MinColor}
  n2 = ScalarParameter() {defaultValue=0.3, parameterName=MaxColor}
  n3 = LinearInterpolate(Alpha=ScalarParameter() {defaultValue=1, parameterName=RandomOffsetPerClient_Mesh})
  n4 = Add(A=Time() {bOverride_Period=true, period=2.6}, B=@n3)
  n5 = Add(A=Time() {bOverride_Period=true, period=1.57}, B=@n3)
  n6 = Constant() {r=0.15}
  n7 = Constant() {r=0.85}
  n8 = Sine(None=Add(A=Time() {bOverride_Period=true, period=1.24}, B=@n3)) {period=1.24}
  n9 = Clamp(
   None=LinearInterpolate(A=Sine(None=@n4) {period=2.6}, B=Sine(None=@n5) {period=1.57}, Alpha=LinearInterpolate(A=@n6, B=@n7, Alpha=@n8) {constA=0.15, constB=0.85})
   Min=ScalarParameter() {group=TurnOff, parameterName=TurnOff_Fading}
  )
  n10 = Clamp(None=@n9)
  n11 = ScalarParameter() {parameterName=HueShift_Mult_Time}
  n12 = ScalarParameter() {defaultValue=0.15, parameterName=HueShift_Mult_Time_Max}
  n13 = Constant()
  n14 = VertexColor()
  n15 = Add(A=VertexColor().R, B=Multiply(A=@n14.G) {constB=0.25})
  n16 = Constant()
  n17 = ScalarParameter() {defaultValue=0.146, parameterName="Speed (S)"}
  n18 = ScalarParameter() {defaultValue=0.5, parameterName="Speed2 (S)"}
  n19 = LinearInterpolate(
   A=@n14.G
   B=Multiply(A=@n14.G, B=ScalarParameter() {defaultValue=0.25, parameterName=Param_1})
   Alpha=LinearInterpolate(
    A=MaterialFunctionCall(Speed=@n17) {materialFunction=&DebugTimeSine}.Result
    B=MaterialFunctionCall(Speed=@n18) {materialFunction=&DebugTimeSine}.Result
   )
  )
  n20 = Time() {bOverride_Period=true, period=5}
  n21 = VertexColor()
  n22 = MaterialFunctionCall(Speed=ScalarParameter() {defaultValue=0.25, parameterName=PanSpeed_Sine}) {materialFunction=&DebugTimeSine}
  n23 = Multiply(
   A=Frac(None=Add(A=@n19, B=@n20))
   B=LinearInterpolate(
    A=ScalarParameter() {defaultValue=2.25, parameterName=TimeMult}
    B=ScalarParameter() {defaultValue=1, parameterName=TimeMult_High}
    Alpha=Multiply(A=@n21.G, B=@n22.Result)
   )
  )
  n24 = ScalarParameter() {defaultValue=0.25, parameterName=Panning_Mask_Radius}
  n25 = ScalarParameter() {defaultValue=0.35, parameterName=Panning_Mask_Radius_Max}
  n26 = Multiply(A=@n9, B=VertexColor().G)
  n27 = Clamp(
   None=SphereMask(
    A=AppendVector(A=@n13, B=@n15)
    B=AppendVector(A=@n16, B=@n23)
    Radius=LinearInterpolate(A=@n24, B=@n25, Alpha=@n26)
    Hardness=ScalarParameter() {defaultValue=0.1, parameterName=Panning_Mask_Hardness}
   )
   Min=ScalarParameter() {group=TurnOff, parameterName=TurnOff_Fading}
  ) {minDefault=1}
  n28 = MaterialFunctionCall(
   Hue Shift Percentage=Multiply(A=LinearInterpolate(A=@n1, B=@n2, Alpha=@n10), B=LinearInterpolate(A=@n11, B=@n12, Alpha=@n27))
   Texture=VectorParameter() {parameterName=VariantColor, defaultValue=<rgba 0 1 5 1>}.RGB
  ) {materialFunction=&HueShift}
  n29 = MaterialFunctionCall(LDR Input=VectorParameter() {parameterName=LDR_Input_Color, defaultValue=<rgba 1 1 1 1>}.RGB, HDR Tint=@n28.Result) {materialFunction=&DeriveHDRfromLDR}
  n30 = ScalarParameter() {defaultValue=0.5, parameterName=TOD_Night_ColorMult}
  n31 = ShadingPathSwitch(
   Default=MaterialFunctionCall(
    Color_Morning=@n29.RemappedRange
    Color_Day=@n29.RemappedRange
    Color_Evening=@n29.RemappedRange
    Color_Night=Multiply(A=@n29.RemappedRange, B=@n30)
   ) {materialFunction=&Color_Time_of_Day}.Result
   Mobile=@n28.Result
  ) {desc=Migrated from FeatureLevelSwitch}
  n32 = ScalarParameter() {defaultValue=3, parameterName=Emissive_Mult_ForMobileAndSwitch}
  n33 = ShadingPathSwitch(Default=ScalarParameter() {defaultValue=0.25, parameterName=Emissive_Mult}, Mobile=@n32) {desc=Migrated from FeatureLevelSwitch}
  MaterialFunctionCall(Emissive=Multiply(A=ShadingPathSwitch(Default=@n31, Forward=@n28.Result), B=ShadingPathSwitch(Default=@n33, Forward=@n32))) {materialFunction=&MF_FX_EmissiveAdjustments}
  n34 = ComponentMask(None=TextureCoordinate()) {r=true}
  n35 = Multiply(A=OneMinus(None=@n34), B=@n34)
  n36 = Clamp(None=Multiply(A=@n35, B=@n35) {constB=2})
  n37 = VectorParameter() {parameterName=Electrical_Tiling, defaultValue=<rgba 1 0.2 0 0>}
  n38 = ComponentMask(None=VectorParameter() {parameterName=Electrical_SpeedValue, defaultValue=<rgba 0 -1.5 0 0>}.RGB) {g=true, r=true}
  n39 = Panner(Coordinate=Multiply(A=TextureCoordinate(), B=ComponentMask(None=@n37.RGB) {g=true, r=true}), Speed=@n38) {speedY=-1}
  n40 = ScalarParameter() {defaultValue=0.25, parameterName="Rotation Angle (0-1) (S)"}
  n41 = VectorParameter() {parameterName=Electrical_Tiling_1, defaultValue=<rgba 1 1 0 0>}
  n42 = Panner(Coordinate=Multiply(A=TextureCoordinate(), B=ComponentMask(None=@n41.RGB) {g=true, r=true}), Speed=@n38) {speedY=-1}
  n43 = ScalarParameter() {defaultValue=0.75, parameterName="Rotation Angle (0-1) (S)_1"}
  n44 = Add(
   A=TextureSample(UVs=MaterialFunctionCall(UVs=@n39, "Rotation Angle (0-1)"=@n40) {materialFunction=&CustomRotator}.Rotated Values) {samplerType=SAMPLERTYPE_LinearColor, texture=/Game/Effects/Fort_Effects/Textures/Tile/T_Raygun_Beam_02}.R
   B=TextureSample(UVs=MaterialFunctionCall(UVs=@n42, "Rotation Angle (0-1)"=@n43) {materialFunction=&CustomRotator}.Rotated Values) {samplerType=SAMPLERTYPE_LinearColor, texture=&T_Raygun_Beam_02}.G
  )
  n45 = Constant()
  n46 = Constant() {r=1}
  n47 = ScalarParameter() {defaultValue=10, parameterName=NearFade}
  n48 = Clamp(None=Power(Base=MaterialFunctionCall(Distance=@n47) {materialFunction=&Near_Camera_Fade}.Result) {constExponent=16})
  n49 = Multiply(A=Multiply(A=Add(A=@n36, B=@n44), B=LinearInterpolate(A=@n45, B=@n46, Alpha=@n48)), B=@n27)
  n50 = ShadingPathSwitch(Default=DepthFade(Opacity=@n49, FadeDistance=ScalarParameter() {defaultValue=25, parameterName=Depth_FadeDistance}), Mobile=@n49) {desc=Migrated from FeatureLevelSwitch}
  Multiply(A=QualitySwitch(Default=ShadingPathSwitch(Default=@n50, Forward=@n49), Low=@n49), B=@n9)
  n51 = Subtract(A=ComponentMask(None=TextureCoordinate()) {g=true}, B=Constant()) {constB=0.5}
  n52 = Clamp(None=Multiply(A=@n51, B=OneMinus(None=@n51)))
  n53 = Multiply(A=DynamicParameter() {paramNames [Thickness,Param2,Param3,Param4]}.Thickness, B=@n52)
  n54 = MaterialFunctionCall(WidthBase=@n53, WidthTip=@n53, UVs for Thickness=TextureCoordinate()) {materialFunction=&SplineThicken}
  n55 = TextureCoordinate() {uTiling=0.002, vTiling=0.1}
  n56 = ScalarParameter() {defaultValue=3, parameterName=TileWpo_Noise}
  n57 = Time()
  n58 = ScalarParameter() {defaultValue=25, parameterName=PanStepSize}
  n59 = Divide(A=Round(None=Multiply(A=@n57, B=@n58)), B=@n58)
  n60 = Constant2Vector() {g=1, r=0.25}
  n61 = ScalarParameter() {defaultValue=-1, parameterName=PanSpeed}
  n62 = TextureSample(UVs=Panner(Coordinate=Multiply(A=@n55, B=@n56), Time=@n59, Speed=Multiply(A=@n60, B=@n61)) {bFractionalPart=true, speedX=0.01, speedY=-0.1}) {samplerType=SAMPLERTYPE_LinearColor, texture=/Game/Global/Textures/Default/Noise/Noise32x32}
  n63 = OneMinus(None=@n62.RGB)
  n64 = ScalarParameter() {defaultValue=2, parameterName=WPO_Noise_Mult_min}
  n65 = ScalarParameter() {defaultValue=5, parameterName=WPO_Noise_Mult_max}
  n66 = Panner(
   Coordinate=Multiply(A=TextureCoordinate() {vTiling=0.1}, B=Multiply(A=@n56) {constB=6}) {constB=0.1}
   Time=Multiply(A=@n59) {constB=0.5}
   Speed=Multiply(A=@n61) {constB=-1.43}
  ) {bFractionalPart=true, speedX=-0.05}
  n67 = Multiply(
   A=Multiply(A=Subtract(A=@n62.RGB, B=@n63), B=LinearInterpolate(A=@n64, B=@n65, Alpha=@n26)) {constA=1}
   B=Add(A=TextureSample(UVs=@n66) {texture=/Game/Global/Textures/Default/Masks_and_Gradients/T_Vert_Gradient_Sections}.R) {constB=0.15}
  )
  n68 = Subtract(A=Subtract(A=WorldPosition().XYZ, B=ActorPositionWS()), B=Multiply(A=ScalarParameter() {defaultValue=1, parameterName=CurlSpeed}, B=Time()))
  n69 = ScalarParameter() {defaultValue=15, parameterName=CurlScale}
  n70 = Multiply(
   A=ScalarParameter() {defaultValue=5, parameterName=CurlIntensity}
   B=VectorNoise(World Position=Divide(A=@n68, B=@n69)) {noiseFunction=VNF_VectorALU}
  )
  n71 = ShadingPathSwitch(Default=ShadingPathSwitch(Default=Multiply(A=@n52, B=@n70), Mobile=@n52) {desc=Migrated from FeatureLevelSwitch}, Forward=@n52)
  Reroute(None=Add(A=Add(A=@n54.WorldPositionOffset, B=@n67), B=QualitySwitch(Default=@n71, Low=@n52)))
 ]
}
