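# Translucent, two-sided lit material for the soft melee swing trail of the Ashton Turbo backpack, usable on beam trails and Niagara ribbons or meshes.
# Sphere masks blend saw, bag and middle colors along the trail, tinted by particle color.
# Flow-map-distorted, panning electric noise combines with light-ray streaks for the transparency detail.
# Edge masks, an optional soft trailing edge, double-sided Fresnel, near-camera fade and particle alpha shape that detail.
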Material /Game/Athena/Items/Gameplay/BackPacks/Ashton/Turbo/FX/M_Ashton_Turbo_Melee_TrailSoft
graph {
 properties {
  bUsedWithBeamTrails=true
  bUsedWithNiagaraMeshParticles=true
  bUsedWithNiagaraRibbons=true
  blendMode=BLEND_Translucent
  pixelDepthOffsetMode=PDOM_Legacy
  translucencyLightingMode=TLM_VolumetricPerVertexDirectional
  twoSided=true
  usedShadingModels=Default Lit
 }
 nodes [
  n1 = VectorParameter() {parameterName=Color_Saw, defaultValue=<rgba 0.25 0.25 0.25 1>}
  n2 = VectorParameter() {parameterName=Color_Bag, defaultValue=<rgba 0.1042 0.7282 1 1>}
  n3 = VectorParameter() {parameterName=TopBot_Mask_Position, defaultValue=<rgba 0 0.5 0 0>}
  n4 = LinearInterpolate(
   A=VectorParameter() {parameterName=Color_Middle, defaultValue=<rgba 0.2182 0.2209 0.25 1>}.RGB
   B=@n1.RGB
   Alpha=SphereMask(
    A=TextureCoordinate() {uTiling=0}
    B=ComponentMask(None=@n3.RGB) {g=true, r=true}
    Radius=ScalarParameter() {defaultValue=0.42, parameterName=TopBot_Mask_Radius}
    Hardness=ScalarParameter() {defaultValue=0.85, parameterName=TopBot_Mask_Hardness}
   )
  )
  n5 = SphereMask(
   A=TextureCoordinate() {uTiling=0}
   B=ComponentMask(None=VectorParameter() {parameterName=Saw_Mask_Position_1, defaultValue=<rgba -0.15 0 0 0>}.RGB) {g=true, r=true}
   Radius=ScalarParameter() {defaultValue=0.3, parameterName=Bag_Mask_Radius}
   Hardness=ScalarParameter() {defaultValue=0.85, parameterName=Bag_Mask_Hardness}
  )
  n6 = TextureCoordinate() {uTiling=0}
  n7 = ComponentMask(None=VectorParameter() {parameterName=Saw_Mask_Position_2, defaultValue=<rgba -0.2 0 0 0>}.RGB) {g=true, r=true}
  n8 = ScalarParameter() {defaultValue=0.6, parameterName=Saw_Mask_Radius_1}
  n9 = ScalarParameter() {defaultValue=0.85, parameterName=Saw_Mask_Hardness_1}
  n10 = ParticleColor()
  n11 = Multiply(
   A=LinearInterpolate(A=@n1.RGB, B=LinearInterpolate(A=@n2.RGB, B=@n4, Alpha=@n5), Alpha=SphereMask(A=@n6, B=@n7, Radius=@n8, Hardness=@n9))
   B=@n10.RGB
  )
  MaterialFunctionCall(Emissive=@n11) {materialFunction=&MF_FX_EmissiveAdjustments}
  n12 = MaterialFunctionCall(Float2=Panner(Coordinate=TextureCoordinate() {vTiling=3.464}) {speedY=0.976}) {materialFunction=&BreakOutFloat2Components}
  n13 = MaterialFunctionCall(X=@n12.G, Y=@n12.G) {materialFunction=&MakeFloat2}
  n14 = MaterialFunctionCall(Float2=Panner(Coordinate=TextureCoordinate() {vTiling=4.4362}) {speedY=-0.7616}) {materialFunction=&BreakOutFloat2Components}
  n15 = MaterialFunctionCall(X=@n14.G, Y=@n14.G) {materialFunction=&MakeFloat2}
  n16 = Add(
   A=Power(Base=TextureSample(UVs=@n13.Result) {texture=/Game/Effects/Fort_Effects/Textures/Outlander/T_InTheZone_LightRays_01}.R) {constExponent=1.5}
   B=Power(Base=TextureSample(UVs=@n15.Result) {texture=&T_InTheZone_LightRays_01}.G)
  )
  n17 = Multiply(
   A=TextureCoordinate() {uTiling=0.5}
   B=AppendVector(A=ScalarParameter() {defaultValue=1, parameterName=MacroTileX}, B=ScalarParameter() {defaultValue=1, parameterName=MacroTileY})
  )
  n18 = Time()
  n19 = ScalarParameter() {defaultValue=6, parameterName=Macro UV Distortion Panning Time}
  n20 = TextureSampleParameter2D(UVs=Panner(Coordinate=@n17, Time=Multiply(A=@n18, B=@n19)) {speedX=0.012, speedY=0.02}) {
   parameterName=Texture-Distortion
   samplerSource=SSM_Wrap_WorldGroupSettings
   samplerType=SAMPLERTYPE_LinearColor
   texture=/Game/Effects/Fort_Effects/Creatures/Elemental/Fire/MergedFireFlowMap
  }
  n21 = Subtract(A=ComponentMask(None=@n20.RGB) {g=true, r=true}) {constB=0.5}
  n22 = Add(A=Multiply(A=Multiply(A=@n21), B=ScalarParameter() {defaultValue=0.25, parameterName=Macro UV Tile Distortion Amount}) {constB=0.3}, B=@n17)
  n23 = AppendVector(A=ScalarParameter() {defaultValue=-0.25, parameterName=Speed X}, B=ScalarParameter() {parameterName=Speed Y})
  n24 = TextureSampleParameter2D(UVs=Panner(Coordinate=@n22, Speed=@n23) {speedX=-0.25}) {
   parameterName=TextureA
   samplerSource=SSM_Wrap_WorldGroupSettings
   samplerType=SAMPLERTYPE_LinearColor
   texture=/Game/Characters/Player/Male/Medium/Bodies/M_MED_SRDrift/Textures/T_ElectricNoise
  }
  n25 = ScalarParameter() {defaultValue=0.5, parameterName=Second itation of glow scale}
  n26 = TextureSampleParameter2D(UVs=Panner(Coordinate=Multiply(A=@n22, B=@n25) {constB=2}, Speed=@n23) {speedX=-0.25}) {
   parameterName=TextureB
   samplerSource=SSM_Wrap_WorldGroupSettings
   samplerType=SAMPLERTYPE_LinearColor
   texture=&T_ElectricNoise
  }
  n27 = Multiply(A=Multiply(A=@n24.G, B=@n26.G), B=ScalarParameter() {defaultValue=5, parameterName=ElectricalPOwer})
  n28 = ComponentMask(None=TextureCoordinate()) {r=true}
  n29 = Add(A=ConstantBiasScale(None=@n27) {bias=-0.5, scale=0.2}, B=OneMinus(None=@n28))
  n30 = LinearInterpolate(
   A=LinearInterpolate(
    A=ScalarParameter() {defaultValue=0.2, parameterName=Fine Streaks Opacity A}
    B=ScalarParameter() {defaultValue=0.3, parameterName=Fine Streaks Opacity B}
    Alpha=Divide(A=@n16)
   ) {constA=1}
   B=ScalarParameter() {defaultValue=0.6, parameterName=Hard Streaks Opacity}
   Alpha=Add(A=Power(Base=@n29) {constExponent=6}, B=@n27)
  ) {constAlpha=0}
  n31 = Constant() {r=0.5}
  n32 = AppendVector(
   A=@n31
   B=MaterialFunctionCall(Float2=TextureCoordinate()) {materialFunction=&BreakOutFloat2Components}.G
  )
  n33 = ScalarParameter() {defaultValue=0.5, parameterName=Mask Radius}
  n34 = ScalarParameter() {defaultValue=0.9, parameterName=Mask Hardness}
  n35 = TextureCoordinate() {uTiling=0.5}
  n36 = Constant2Vector() {g=0.82, r=0.85}
  n37 = Constant() {r=0.8}
  n38 = Constant() {r=0.3}
  n39 = Power(Base=OneMinus(None=SphereMask(A=@n35, B=@n36, Radius=@n37, Hardness=@n38))) {constExponent=5}
  n40 = TextureCoordinate()
  n41 = Power(Base=OneMinus(None=MaterialFunctionCall(Float2=@n40) {materialFunction=&BreakOutFloat2Components}.R))
  n42 = Multiply(
   A=Multiply(A=Multiply(A=@n27, B=@n30), B=SphereMask(A=@n32, B=@n31, Radius=@n33, Hardness=@n34))
   B=StaticSwitchParameter(True=Multiply(A=@n39, B=@n41), False=@n39) {parameterName=SoftTrailingEdge}
  )
  n43 = Clamp(None=@n42)
  n44 = ScalarParameter() {parameterName=Clamp Detail Values}
  n45 = Multiply(A=@n28, B=ScalarParameter() {defaultValue=3, parameterName=Leading Edge Mask Multiply}) {constB=3}
  n46 = Multiply(
   A=Multiply(A=LinearInterpolate(A=@n42, B=@n43, Alpha=@n44), B=Saturate(None=@n45))
   B=MaterialFunctionCall(Distance=ScalarParameter() {parameterName=NearFadeDistance}) {materialFunction=&Near_Camera_Fade}.Result
  )
  n47 = MaterialFunctionCall(
   Brightness=ScalarParameter() {defaultValue=1.5, parameterName=Fresnel Brightness}
   Exponent=ScalarParameter() {defaultValue=3, parameterName=Fresnel Exponent}
  ) {materialFunction=&MF_Fresnel_FX}
  n48 = ComponentMask(None=TextureCoordinate()) {r=true}
  n49 = ScalarParameter() {defaultValue=2, group=LeadingEdgeMaskB, parameterName=LeadingEdgeMaskB-Power}
  n50 = Multiply(
   A=Saturate(None=Power(Base=@n48, Exp=@n49))
   B=ScalarParameter() {defaultValue=4, group=LeadingEdgeMaskB, parameterName=LeadingEdgeMaskB-Multiply}
  ) {constB=3}
  n51 = OneMinus(None=ScalarParameter() {group=LeadingEdgeMaskB, parameterName=LeadingEdgeMaskB-Influence})
  Multiply(A=Multiply(A=@n10.A, B=Multiply(A=@n46, B=@n47.Fresnel Double Sided)), B=Saturate(None=Add(A=@n50, B=@n51)))
  Reroute(None=@n11)
 ]
}
