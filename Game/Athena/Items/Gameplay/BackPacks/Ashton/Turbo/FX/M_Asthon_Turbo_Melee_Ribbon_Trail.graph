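# Translucent, unlit, two-sided ribbon trail material for the Ashton Turbo backpack's melee effect, drawing an electric streak.
# Emissive comes from particle color, scaled by time of day and passed through HDR remapping and FX adjustment functions.
# A flow map distorts two panning electric noise samples, which are multiplied, sharpened and masked by trail texture, particle alpha and U-gradient falloffs.
# The resulting opacity gets depth fade, except on the mobile path.
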Material /Game/Athena/Items/Gameplay/BackPacks/Ashton/Turbo/FX/M_Asthon_Turbo_Melee_Ribbon_Trail
graph {
 properties {
  bUsedWithBeamTrails=true
  bUsedWithNiagaraRibbons=true
  blendMode=BLEND_Translucent
  pixelDepthOffsetMode=PDOM_Legacy
  shadingModel=MSM_Unlit
  twoSided=true
  usedShadingModels=Unlit
 }
 nodes [
  n1 = ParticleColor()
  n2 = Multiply(A=@n1.RGB) {constB=1.5}
  n3 = Multiply(A=@n1.RGB) {constB=2.5}
  MaterialFunctionCall(Emissive=MaterialFunctionCall(
    LDR Input=Constant3Vector() {constant=<rgba 1 1 1 0>}
    HDR Tint=MaterialFunctionCall(Color_Morning=@n2, Color_Day=@n3, Color_Evening=@n2, Color_Night=@n1.RGB) {materialFunction=&Color_Time_of_Day}.Result
   ) {materialFunction=&DeriveHDRfromLDR}.RemappedRange) {materialFunction=&MF_FX_EmissiveAdjustments}
  n4 = ComponentMask(None=TextureCoordinate()) {r=true}
  n5 = OneMinus(None=Power(Base=OneMinus(None=@n4)) {constExponent=0.5})
  n6 = Multiply(A=TextureCoordinate() {vTiling=0.25}, B=ScalarParameter() {defaultValue=1, parameterName=MacroTile})
  n7 = Panner(Coordinate=@n6, Time=Multiply(A=Time(), B=ScalarParameter() {defaultValue=6, parameterName=Macro UV Distortion Panning Time})) {speedX=0.012, speedY=0.02}
  n8 = Subtract(A=ComponentMask(None=TextureSample(UVs=@n7) {samplerType=SAMPLERTYPE_LinearColor, texture=/Game/Effects/Fort_Effects/Creatures/Elemental/Fire/MergedFireFlowMap}.RGB) {g=true, r=true}) {constB=0.5}
  n9 = Add(A=Multiply(A=Multiply(A=@n8), B=ScalarParameter() {defaultValue=0.25, parameterName=Macro UV Tile Distortion Amount}) {constB=0.3}, B=@n6)
  n10 = Multiply(A=@n9, B=ScalarParameter() {defaultValue=0.5, parameterName=Second itation of glow scale}) {constB=2}
  n11 = Multiply(
   A=TextureSample(UVs=Panner(Coordinate=@n9) {speedX=1}) {samplerType=SAMPLERTYPE_LinearColor, texture=/Game/Characters/Player/Male/Medium/Bodies/M_MED_SRDrift/Textures/T_ElectricNoise}.G
   B=TextureSample(UVs=Panner(Coordinate=@n10) {speedX=1}) {samplerType=SAMPLERTYPE_LinearColor, texture=&T_ElectricNoise}.G
  )
  n12 = ScalarParameter() {defaultValue=10, parameterName=ElectricaIntensity}
  n13 = Multiply(
   A=Multiply(
    A=Multiply(A=@n1.A, B=@n5)
    B=TextureSampleParameter2D() {parameterName=TrailMask, texture=/Game/Weapons/FORT_Melee/Effects/LaserSwordFounders/T_LaserSwordTrail}.RGB
   )
   B=Power(Base=Multiply(A=@n11, B=@n12), Exp=ScalarParameter() {defaultValue=3, parameterName=ElectricalPOwer})
  )
  n14 = DotProduct(A=Normalize(VectorInput=CameraVectorWS()), B=Normalize(VectorInput=VertexNormalWS()))
  Multiply(A=@n13, B=Power(Base=Abs(None=@n14), Exp=Constant() {r=3}))
  n15 = ComponentMask(None=TextureCoordinate()) {r=true}
  n16 = Multiply(A=OneMinus(None=Power(Base=@n15) {constExponent=1}), B=@n13)
  ShadingPathSwitch(Default=DepthFade(Opacity=@n16, FadeDistance=ScalarParameter() {defaultValue=250, parameterName=FadeDistance}), Mobile=@n16) {desc=Migrated from FeatureLevelSwitch}
 ]
}
