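# Niagara effect for the melee swing of the Ashton Turbo back pack, drawing a blue glowing trail with an axe mesh.
# Two short-lived ribbon emitters with curve-driven color and alpha form the trail, the second at very low base alpha.
# Single bursts add the axe mesh, which fades in and out, and a flashing light with sprite.
# Two flame sprite emitters, tinted blue and violet, rise with drag and flipbook animation.
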NiagaraSystem /Game/Athena/Items/Gameplay/BackPacks/Ashton/Turbo/FX/NS_Ashton_Turbo_MeleeSwing_Ribbon
dependencies {
 usedDataInterfaces [
  &NiagaraDataInterfaceColorCurve
  &NiagaraDataInterfaceCurve
  &NiagaraDataInterfaceSpriteRendererInfo
 ]
 usedDynamicInputs []
 usedModules [&SystemState]
}
emitters {
 Axe_Mesh {
  emitterUpdateScript [
   &EmitterState()
   &SpawnBurst_Instantaneous(Use Spawn Probability=false, UseLoopCountLimit=false)
  ]
  particleSpawnScript [
   &InitializeParticle(
    Color=<rgba 1 1 1 100>
    Mesh Scale=<100 90 100>
    Mesh Scale Mode=::NewEnumerator4
    Position Offset=<-1 0 -5>
    Position Offset Coordinate Space=::Local
    UsePositionOffset=true
    Write Material Random=true
   )
   &CameraOffset(Camera Offset Amount=5)
   &DynamicMaterialParameters(
    Index 0 Param 1=1
    Index 0 Param 2=1
    Module.Param0WriteEnabled=true
    Module.Param1WriteEnabled=true
    Module.Param2WriteEnabled=false
    Module.Param3WriteEnabled=false
   )
   &InitialMeshOrientation(Mesh Orientation Mode=::NewEnumerator7, Rotation=<-0.75 0 0>, Use Rotation Vector=true)
  ]
  particleUpdateScript [
   &ParticleState()
   &ScaleColor(
    Scale Alpha=&FloatFromCurve(FloatCurve=DataInterface() {propertyValues {Curve {keys [
         <at 0 0 RCIM_Cubic>
         {arriveTangent=1.6719, interpMode=RCIM_Cubic, leaveTangent=1.6719, time=0.2843, value=0.9747}
         {arriveTangent=-1.3618, interpMode=RCIM_Cubic, leaveTangent=-1.3618, time=0.588, value=0.9831}
         <at 1 0 RCIM_Cubic>
        ]}}})
    ScaleA=true
    ScaleRGB=false
   )
  ]
  rendererClasses [&NiagaraMeshRendererProperties]
  renderers [{rendererClass=&NiagaraMeshRendererProperties, rendererIndex=0}]
 }
 Flames {
  emitterUpdateScript [&EmitterState(Life Cycle Mode=::NewEnumerator1, Loop Behavior=::NewEnumerator1, Loop Duration=0.15, UseLoopDelay=false)]
  particleSpawnScript [
   &InitializeParticle(
    Color=<rgba 1.055 1.5834 3 0.125>
    Lifetime Max=0.35
    Lifetime Min=0.2
    Lifetime Mode=::NewEnumerator1
    Position Offset=&RandomRangeVector(Minimum=<0 0 25>, Maximum=<0 0 35>)
    Sprite Rotation Mode=::NewEnumerator2
    Sprite Size Mode=::NewEnumerator1
    Uniform Sprite Size Max=40
    Uniform Sprite Size Min=20
    UsePositionOffset=true
    Write Material Random=true
   )
   &CameraOffset(Camera Offset Amount=&RandomRangeFloat(Minimum=-25, Maximum=50))
  ]
  particleUpdateScript [
   &ParticleState()
   &ScaleSpriteSize(Uniform Curve Sprite Scale=DataInterface() {propertyValues {Curve {keys [<at 0 0 RCIM_Cubic>, <at 0.35 1>, {tangentMode=RCTM_User, time=1, value=0}]}}})
   &SubUVAnimation(
    Direct Frame Index:NiagaraFloat=&FloatFromCurve(FloatCurve=DataInterface() {propertyValues {Curve {keys [<at 0 32>,<at 1 63>]}}})
    "Sprite Renderer "=DataInterface() {propertyValues {SpriteRenderer=&self:Flames_5.NiagaraSpriteRendererProperties_0}}
    SubUV Animation Mode:ENiagara_SubUVLookupModeV2=::NewEnumerator5
   )
   &AccelerationForce(Acceleration=<0 0 150>)
   &Drag(Use Linear Drag=true, Use Rotational Drag=true)
   &SolveForcesAndVelocity(Clamp Velocity=false, Limit Acceleration=false)
   &ScaleColor(
    Scale Alpha=&FloatFromCurve(FloatCurve=DataInterface() {propertyValues {Curve {keys [<at 0 0 RCIM_Cubic>, {interpMode=RCIM_Cubic, leaveTangent=-1.5277, tangentMode=RCTM_Break, time=0.0389, value=1}, <at 1 0 RCIM_Cubic>]}}})
    Scale RGB=&MakeVectorFromLinearColorRGB(Color=&ColorFromCurve(ColorCurve=DataInterface() {propertyValues {
        AlphaCurve {keys []}
        BlueCurve {keys [<at 0 3>,<at 1 1>]}
        GreenCurve {keys [<at 0 1.4>,<at 1 0.3333>]}
        RedCurve {keys [<at 0 0>,<at 1 0>]}
       }}))
    ScaleA=true
    ScaleRGB=true
   )
   &DynamicMaterialParameters(
    Index 0 Param 2=1
    Index 0 Param 4=&RandomRangeFloat(Minimum=0.6, Maximum=1)
    Module.Param0WriteEnabled=false
    Module.Param1WriteEnabled=true
    Module.Param2WriteEnabled=false
    Module.Param3WriteEnabled=true
   )
  ]
  rendererClasses [&NiagaraSpriteRendererProperties]
  renderers [{rendererClass=&NiagaraSpriteRendererProperties}]
 }
 Flames001 {
  emitterUpdateScript [&EmitterState(Life Cycle Mode=::NewEnumerator1, Loop Behavior=::NewEnumerator1, Loop Duration=0.25, UseLoopDelay=false)]
  particleSpawnScript [
   &InitializeParticle(
    Color=<rgba 0.5253 0.063 3 0.5>
    Lifetime Max=0.4
    Lifetime Min=0.25
    Lifetime Mode=::NewEnumerator1
    Position Offset=&RandomRangeVector(Minimum=<-25 0 25>, Maximum=<25 0 35>)
    Sprite Rotation Mode=::NewEnumerator2
    Sprite Size Mode=::NewEnumerator1
    Uniform Sprite Size Max=40
    Uniform Sprite Size Min=25
    UsePositionOffset=true
    Write Material Random=true
   )
   &CameraOffset(Camera Offset Amount=&RandomRangeFloat(Minimum=-25, Maximum=50))
  ]
  particleUpdateScript [
   &ParticleState()
   &ScaleSpriteSize(Uniform Curve Sprite Scale=DataInterface() {propertyValues {Curve {keys [<at 0 0 RCIM_Cubic>, <at 0.35 1>, {tangentMode=RCTM_User, time=1, value=0}]}}})
   &SubUVAnimation(
    Direct Frame Index:NiagaraFloat=&FloatFromCurve(FloatCurve=DataInterface() {propertyValues {Curve {keys [<at 0 50>,<at 1 63>]}}})
    "Sprite Renderer "=DataInterface() {propertyValues {SpriteRenderer=&self:Flames001_7.NiagaraSpriteRendererProperties_0}}
    SubUV Animation Mode:ENiagara_SubUVLookupModeV2=::NewEnumerator5
   )
   &AccelerationForce(Acceleration=<0 0 150>)
   &Drag(Use Linear Drag=true, Use Rotational Drag=true)
   &SolveForcesAndVelocity(Clamp Velocity=false, Limit Acceleration=false)
   &ScaleColor(
    Scale Alpha=&FloatFromCurve(FloatCurve=DataInterface() {propertyValues {Curve {keys [<at 0 0 RCIM_Cubic>, {interpMode=RCIM_Cubic, leaveTangent=-1.5277, tangentMode=RCTM_Break, time=0.0389, value=1}, <at 1 0 RCIM_Cubic>]}}})
    Scale RGB=&MakeVectorFromLinearColorRGB(Color=&ColorFromCurve(ColorCurve=DataInterface() {propertyValues {
        AlphaCurve {keys []}
        BlueCurve {keys [<at 0 3>,<at 1 1>]}
        GreenCurve {keys [<at 0 1.4>,<at 1 0.3333>]}
        RedCurve {keys [<at 0 0>,<at 1 0>]}
       }}))
    ScaleA=true
    ScaleRGB=true
   )
   &DynamicMaterialParameters(
    Index 0 Param 2=1
    Index 0 Param 4=&RandomRangeFloat(Minimum=0.6, Maximum=1)
    Module.Param0WriteEnabled=false
    Module.Param1WriteEnabled=true
    Module.Param2WriteEnabled=false
    Module.Param3WriteEnabled=true
   )
  ]
  rendererClasses [&NiagaraSpriteRendererProperties]
  renderers [{rendererClass=&NiagaraSpriteRendererProperties}]
 }
 Light {
  emitterUpdateScript [
   &EmitterState()
   &SpawnBurst_Instantaneous(Use Spawn Probability=false, UseLoopCountLimit=false)
  ]
  particleSpawnScript [&InitializeParticle(
    Color=<rgba 1 1 1 32>
    Lifetime=0.35
    Position Offset=<0 0 50>
    Position Offset Coordinate Space=::Local
    Sprite Size Mode=::NewEnumerator3
    Uniform Sprite Size=0
    UsePositionOffset=true
    Write Material Random=true
   )]
  particleUpdateScript [
   &ParticleState()
   &ScaleColor(
    Scale Alpha=&FloatFromCurve(FloatCurve=DataInterface() {propertyValues {Curve {keys [
         <at 0 0 RCIM_Cubic>
         {arriveTangent=-1.8836, interpMode=RCIM_Cubic, leaveTangent=-1.8836, tangentMode=RCTM_User, time=0.1474, value=1}
         {arriveTangent=-1.3496, interpMode=RCIM_Cubic, leaveTangent=-1.3496, time=0.2944, value=0.2324}
         {arriveTangent=-0.3294, interpMode=RCIM_Cubic, leaveTangent=-0.3294, time=0.8126, value=0.1022}
         <at 1 0 RCIM_Cubic>
        ]}}})
    Scale RGB=&MakeVectorFromLinearColorRGB(Color=&ColorFromCurve(ColorCurve=DataInterface() {propertyValues {
        AlphaCurve {keys []}
        BlueCurve {keys [
          <at 0 25 RCIM_Cubic>
          {arriveTangentWeight=3.8197, interpMode=RCIM_Cubic, leaveTangentWeight=12.6837, tangentMode=RCTM_User, time=0.231, value=5}
          <at 1 0.5 RCIM_Cubic>
         ]}
        GreenCurve {keys [
          <at 0 14.9754 RCIM_Cubic>
          {arriveTangentWeight=1.4036, interpMode=RCIM_Cubic, leaveTangentWeight=4.6608, tangentMode=RCTM_User, time=0.231, value=1.0458}
          <at 1 0 RCIM_Cubic>
         ]}
        RedCurve {keys [
          <at 0 1.3536 RCIM_Cubic>
          {arriveTangentWeight=0.1647, interpMode=RCIM_Cubic, leaveTangentWeight=0.547, tangentMode=RCTM_User, time=0.231, value=1.0072}
          <at 1 0.033 RCIM_Cubic>
         ]}
       }}))
    ScaleA=true
    ScaleRGB=true
   )
   /SetVariables(Particles.LightExponent=16, Particles.LightRadius=150) {isSetParameters=true}
  ]
  rendererClasses [&NiagaraSpriteRendererProperties, &NiagaraLightRendererProperties]
  renderers [
   {rendererClass=&NiagaraSpriteRendererProperties}
   {rendererClass=&NiagaraLightRendererProperties, rendererIndex=1}
  ]
 }
 Main_Ribbon {
  emitterUpdateScript [
   &EmitterState(Life Cycle Mode=::NewEnumerator1, Loop Behavior=::NewEnumerator1, Loop Duration=0.3, UseLoopDelay=false)
   &SpawnRate(SpawnRate=50, Use Spawn Probability=false)
  ]
  particleSpawnScript [&InitializeParticle(
    Lifetime=0.5
    Position Offset=<0 0 25>
    Position Offset Coordinate Space=::Local
    Ribbon Width=10
    Ribbon Width Mode=::NewEnumerator1
    UsePositionOffset=true
    Write Material Random=true
   )]
  particleUpdateScript [
   &ParticleState()
   &ScaleColor(
    Scale Alpha=&FloatFromCurve(FloatCurve=DataInterface() {propertyValues {Curve {keys [
         <at 0 0 RCIM_Cubic>
         {arriveTangent=-1.8836, interpMode=RCIM_Cubic, leaveTangent=-1.8836, tangentMode=RCTM_User, time=0.1474, value=1}
         {arriveTangent=-1.3496, interpMode=RCIM_Cubic, leaveTangent=-1.3496, time=0.2944, value=0.2324}
         {arriveTangent=-0.3294, interpMode=RCIM_Cubic, leaveTangent=-0.3294, time=0.8126, value=0.1022}
         <at 1 0 RCIM_Cubic>
        ]}}})
    Scale RGB=&MakeVectorFromLinearColorRGB(Color=&ColorFromCurve(ColorCurve=DataInterface() {propertyValues {
        AlphaCurve {keys []}
        BlueCurve {keys [
          <at 0 50 RCIM_Cubic>
          {arriveTangentWeight=3.8197, interpMode=RCIM_Cubic, leaveTangentWeight=12.6837, tangentMode=RCTM_User, time=0.2314, value=5}
          <at 1 0.5 RCIM_Cubic>
         ]}
        GreenCurve {keys [
          <at 0 18.1658 RCIM_Cubic>
          {arriveTangentWeight=1.4036, interpMode=RCIM_Cubic, leaveTangentWeight=4.6608, tangentMode=RCTM_User, time=0.2314, value=1.0458}
          <at 1 0 RCIM_Cubic>
         ]}
        RedCurve {keys [
          <at 0 1.9194 RCIM_Cubic>
          {arriveTangentWeight=0.1647, interpMode=RCIM_Cubic, leaveTangentWeight=0.547, tangentMode=RCTM_User, time=0.2314, value=1.0072}
          <at 1 0.033 RCIM_Cubic>
         ]}
       }}))
    ScaleA=true
    ScaleRGB=true
   )
  ]
  rendererClasses [&NiagaraRibbonRendererProperties]
  renderers [{rendererClass=&NiagaraRibbonRendererProperties}]
 }
 Main_Ribbon001 {
  emitterUpdateScript [
   &EmitterState(Life Cycle Mode=::NewEnumerator1, Loop Behavior=::NewEnumerator1, Loop Duration=0.3, UseLoopDelay=false)
   &SpawnRate(SpawnRate=50, Use Spawn Probability=false)
  ]
  particleSpawnScript [&InitializeParticle(
    Color=<rgba 1 1 1 0.02>
    Lifetime=0.5
    Position Offset=<0 0 25>
    Position Offset Coordinate Space=::Local
    Ribbon Width=10
    Ribbon Width Mode=::NewEnumerator1
    UsePositionOffset=true
    Write Material Random=true
   )]
  particleUpdateScript [
   &ParticleState()
   &ScaleColor(
    Scale Alpha=&FloatFromCurve(FloatCurve=DataInterface() {propertyValues {Curve {keys [
         <at 0 0 RCIM_Cubic>
         {arriveTangent=-1.8836, interpMode=RCIM_Cubic, leaveTangent=-1.8836, tangentMode=RCTM_User, time=0.1474, value=1}
         {arriveTangent=-1.3496, interpMode=RCIM_Cubic, leaveTangent=-1.3496, time=0.2944, value=0.2324}
         {arriveTangent=-0.3294, interpMode=RCIM_Cubic, leaveTangent=-0.3294, time=0.8126, value=0.1022}
         <at 1 0 RCIM_Cubic>
        ]}}})
    Scale RGB=&MakeVectorFromLinearColorRGB(Color=&ColorFromCurve(ColorCurve=DataInterface() {propertyValues {
        AlphaCurve {keys []}
        BlueCurve {keys [
          <at 0 25 RCIM_Cubic>
          {arriveTangentWeight=3.8197, interpMode=RCIM_Cubic, leaveTangentWeight=12.6837, tangentMode=RCTM_User, time=0.231, value=5}
          <at 1 0.5 RCIM_Cubic>
         ]}
        GreenCurve {keys [
          <at 0 14.9754 RCIM_Cubic>
          {arriveTangentWeight=1.4036, interpMode=RCIM_Cubic, leaveTangentWeight=4.6608, tangentMode=RCTM_User, time=0.231, value=1.0458}
          <at 1 0 RCIM_Cubic>
         ]}
        RedCurve {keys [
          <at 0 1.3536 RCIM_Cubic>
          {arriveTangentWeight=0.1647, interpMode=RCIM_Cubic, leaveTangentWeight=0.547, tangentMode=RCTM_User, time=0.231, value=1.0072}
          <at 1 0.033 RCIM_Cubic>
         ]}
       }}))
    ScaleA=true
    ScaleRGB=true
   )
  ]
  rendererClasses [&NiagaraRibbonRendererProperties]
  renderers [{rendererClass=&NiagaraRibbonRendererProperties}]
 }
}
summary {
 systemName=NS_Ashton_Turbo_MeleeSwing_Ribbon
 emitters [
  {emitterName=Main_Ribbon, rendererClasses [&NiagaraRibbonRendererProperties]}
  {emitterName=Main_Ribbon001, rendererClasses [&NiagaraRibbonRendererProperties]}
  {emitterName=Light, rendererClasses [&NiagaraSpriteRendererProperties, &NiagaraLightRendererProperties]}
  {emitterName=Axe_Mesh, rendererClasses [&NiagaraMeshRendererProperties]}
  {emitterName=Flames, rendererClasses [&NiagaraSpriteRendererProperties]}
  {emitterName=Flames001, rendererClasses [&NiagaraSpriteRendererProperties]}
 ]
 userVariables []
}
