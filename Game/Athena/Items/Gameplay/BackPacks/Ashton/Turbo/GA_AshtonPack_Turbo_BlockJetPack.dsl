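; Gameplay ability for the Ashton Turbo backpack that, by its name, blocks jetpack use, though no blocking tags appear here.
; On activation it removes owner gameplay effects granting Event.IsUsingJetpack.
; If the player pawn is not walking, it waits for a movement mode change to Walking before ending; otherwise it ends at once.
; The failed-activation event only calls the parent, and the end-ability event is empty.
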(event Ability|EventActivateAbility
  (Ability|RemoveGameplayEffectFromOwnerWithGrantedTags "(GameplayTags=((TagName=\"Event.IsUsingJetpack\")))")
  (bind _asfort_player_pawn (Utilities|Casting|CastToFortPlayerPawn (Ability|GetAvatarActorfromActorInfo)))
  (if (not (Utilities|Enum|Equal(Enum) (Variables|/Script/Engine.CharacterMovementComponent|GetMovementMode (Variables|/Script/Engine.Character|GetCharacterMovement _asfort_player_pawn)) "MOVE_Walking"))
    (Ability|Tasks|WaitMovementModeChange "MOVE_Walking"
      (:OnChange
        _)))
  (Ability|EndAbility))

(event Ability|EventFailedToActivatePassiveAbility
  (params (ActorInfo :type Gameplay%20Ability%20Actor%20Info%20Structure))
  (|CallParentFunction))

(event Ability|EventOnEndAbility
  (params (bWasCancelled :type Boolean)))
