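; Gameplay ability for the Ashton Turbo backpack that throws a hammer projectile, or sends a recall event instead when the block-throw tag is present.
; A throw applies a windup effect, clears it immediately in ThrowProjectile, then plays the montage and waits for its trigger.
; On trigger it plays a throw cue; the server commits to prevent multi-spawn exploits, spawns the projectile, sends its reference to the player by gameplay event and completes after a short delay.
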(event Ability|EventActivateAbility
  (bind (_owneractor _avataractor) (Utilities|Struct|BreakGameplayAbilityActorInfo (Ability|GetActorInfo)))
  ;[1 1 1] EVENT -> Activate Ability
  (bind _asfort_player_pawn (Utilities|Casting|CastToFortPlayerPawn _avataractor)
    (:then
      (bind _output_get (Variables|ReferencedActors|SetPlayerPawn _asfort_player_pawn))
      ;[1 1 1] Check if we should Recall Hammer or throw a new one
      (if (GameplayTags|HasMatchingGameplayTag (Variables|ReferencedActors|GetPlayerPawn) (Variables|Defaults|GetBlockThrowTag))
        (Ability|SendGameplayEventtoActor _output_get "(TagName=\"Event.Athena.Item.Ashton.Turbo.Recall\")")
        (Ability|EndAbility)
        (else
          (ParentDefaults|PrethrowSetup) ; Ability is instanced per actor, so we need to reset variables\nevery time the ability is activated.
          (CallFunction|ThrowProjectile))))
    (:CastFailed
      (Ability|EndAbility))))

(event Custom|ThrowProjectile
  ;[1 1 1] Throw Projectile
  (Ability|SetShouldBlockOtherAbilities true)
  (ParentDefaults|PostThrowCleanup)
  ;[1 1 1] Projectile is spawned when the montage reaches ability notify
  (Ability|Tasks|PlayMontageandWaitforTrigger "None" (Variables|Animation|GetThrowMontage) 1.2
    (:Triggered
      (CallFunction|AttemptSpawnThrownProjectile))
    (:Cancelled
      (Ability|EndAbility))
    (:Completed
      (Ability|SetShouldBlockOtherAbilities false))))

(event Custom|AttemptSpawnThrownProjectile
  (bind (_location _rotation) (ParentDefaults|PrjSpawnLocationandRotation))
  ;[1 1 1] Attempt to spawn projectile (only on server)
  (Ability|ExecuteGameplayCueOnOwner (Variables|Defaults|GetGC_OnThrow)) ; GC to play on the player when the projectile is thrown
  (if (Networking|IsServer)
    ;[1 1 1] Commit the ability here to prevent exploits relating to spawning multiple projectiles
    (bind _returnvalue (Ability|CommitAbility))
    (if _returnvalue
      ;[1 1 1] Spawn Projectile!
      (Variables|Defaults|SetExplosionEffectContainerSpec (Ability|MakeGameplayEffectContainerSpec (Variables|Defaults|GetEC_DefaultExplosion)))
      (bind _spawnedprojectile (ParentDefaults|SpawnThrownProjectile (Variables|ReferencedActors|GetProj_ThrownProjectile) _location _rotation (Variables|Defaults|GetExplosionEffectContainerSpec) (ScalableFloat|GetValueatLevel (Variables|Defaults|GetProjectileSpeed))))
      (Variables|ReferencedActors|SetProjectileRef _spawnedprojectile)
      ;[1 1 1] Send Event to track projectile, then end ability
      (CallFunction|SendProjReferencetoOwner _spawnedprojectile)
      (Ability|Tasks|WaitDelay 0.4
        (:OnFinish
          (Ability|AbilityCompleted) ; We only execute this part of the event on the server,\nso is this meaningfully distinguishable from EndAbility?
        )) ; Would prefer to end the ability without a delay if possible. (Chris M.)
      (else
        (Ability|EndAbility)))))

(event Ability|EventOnEndAbility
  (params (bWasCancelled :type Boolean))
  ;[1 1 1] EVENT -> End Ability
  (|CallParentFunction bWasCancelled)
  (ParentDefaults|PostThrowCleanup))

(fn GetOwningPlayer
  (bind (_owneractor _avataractor) (Utilities|Struct|BreakGameplayAbilityActorInfo (Ability|GetActorInfo)))
  (Utilities|IsValid (Variables|ReferencedActors|GetPlayerPawn)
    (:Is%20Valid
      (return (Variables|ReferencedActors|GetPlayerPawn)))
    (:Is%20Not%20Valid
      (bind _asfort_player_pawn (Utilities|Casting|CastToFortPlayerPawn _avataractor)
        (:then
          (return _asfort_player_pawn))
        (:CastFailed
          (return 0))))))

(fn GetProjectileTrajectoryActor
  (return 0))

(fn GetProjectileTrajectoryPoints
  (return 0 0 0))

(fn OnProjectileSetup)

(fn PostThrowCleanup
  (Variables|AbilityState|SetInThrowWindup false)
  (Ability|RemoveGameplayEffectFromOwnerWithHandle (Variables|Defaults|GetHandle_GE_InWindup)))

(fn PrethrowSetup
  ;[1 1 1] We allow a player to cancel out of the windup for most thrown projectiles
  (Ability|SetShouldBlockOtherAbilities false)
  (Ability|SetCanbeCanceled true)
  ;[1 1 1] The input is being held down and the actor is winding up to throw the projectile
  (Variables|AbilityState|SetAbilityKeyPressed true)
  (Variables|AbilityState|SetInThrowWindup true)
  ;[1 1 1] Add a GE that applies looping GC while in the windup
  (bind _returnvalue (Ability|ApplyGameplayEffectToOwner (Variables|Defaults|GetGE_InWindup)))
  (Variables|Defaults|SetHandle_GE_InWindup _returnvalue))

(fn PrjSpawnLocationAndRotation
  (bind (_outlocation _outrotation) (Actor|GetActorEyesViewPoint (ParentDefaults|GetOwningPlayer)))
  (return (+ (Transformation|GetActorLocation (Variables|ReferencedActors|GetPlayerPawn)) (Math|Vector|vector+vector (* (Math|Vector|GetForwardVector (Pawn|GetControlRotation (Variables|ReferencedActors|GetPlayerPawn) :!class /Script/Engine.Pawn)) (Variables|SpawnedProjOffsets|GetSpawnPrjForwardOffset)) (* (Math|Vector|GetRightVector (Pawn|GetControlRotation (Variables|ReferencedActors|GetPlayerPawn) :!class /Script/Engine.Pawn)) (Variables|SpawnedProjOffsets|GetSpawnRightOffset)) (select (Variables|/Script/Engine.Character|GetbIsCrouched (Variables|ReferencedActors|GetPlayerPawn)) (* (Math|Vector|GetUpVector (Pawn|GetControlRotation (Variables|ReferencedActors|GetPlayerPawn) :!class /Script/Engine.Pawn)) (Variables|SpawnedProjOffsets|GetSpawnUpOffsetCrouch)) (* (Math|Vector|GetUpVector (Pawn|GetControlRotation (Variables|ReferencedActors|GetPlayerPawn) :!class /Script/Engine.Pawn)) (Variables|SpawnedProjOffsets|GetSpawnUpOffset))))) _outrotation))

(fn SendProjReferenceToOwner
  (params (SpawnedProjectile :type Fort%20Projectile%20Base%20Object%20Reference))
  (Ability|SendGameplayEventtoActor (Variables|ReferencedActors|GetPlayerPawn) "(TagName=\"Event.Athena.Item.Ashton.Turbo.SpawnProj\")" (Utilities|Struct|MakeGameplayEventData "(TagName=\"Event.Athena.Item.Ashton.Turbo.SpawnProj\")" (Variables|ReferencedActors|GetPlayerPawn) (Variables|ReferencedActors|GetPlayerPawn) SpawnedProjectile)))

(fn ShouldOnlyShowTrajectoryOnUse
  (return false))

(fn SpawnThrownProjectile
  (params (ProjectileClass :type Fort%20Projectile%20Base%20Class%20Reference) (ProjectileSpawnLocation :type Vector%20%28by%20ref%29) (ProjectileSpawnDirection :type Rotator%20%28by%20ref%29) (ExplosionEffectContainerSpec :type Fort%20Gameplay%20Effect%20Container%20Spec%20Structure) (ProjectileInitialSpeed :type Float%20%28double-precision%29) (ProjectileGravityScale :type Float%20%28double-precision%29))
  (bind _returnvalue (Projectile|SpawnProjectile :ProjectileClass ProjectileClass :RequestedBy (ParentDefaults|GetOwningPlayer) :SpawnLocation ProjectileSpawnLocation :SpawnRotation ProjectileSpawnDirection :SpawnDirection ProjectileSpawnDirection :EffectContainerSpecToApplyOnExplode ExplosionEffectContainerSpec :bUseDefaultPhysics false :InitialSpeed ProjectileInitialSpeed :GravityScale ProjectileGravityScale :ChargePercent 0.0))
  (bind _output_get (Variables|ReferencedActors|SetSpawnedProjectile _returnvalue))
  (return _output_get))

(fn SpawnTrajectoryIndicator
  (params (bSpawnOnEquip :type Boolean))
  (return 0))
