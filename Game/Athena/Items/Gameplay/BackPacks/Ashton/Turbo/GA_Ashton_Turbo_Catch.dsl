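; GA_Ashton_Turbo_Catch is a gameplay ability for the catch action of the Ashton Turbo backpack item.
; Its activation event holds only a disabled parent call, so this graph adds no activation logic of its own.
; On end, it calls the parent, then removes owner gameplay effects granting the Turbo ThrowBlock tag as a fail-safe against a lingering block.
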(event Ability|EventActivateAbility
  (|CallParentFunction) ; This node is disabled and will not be called.\nDrag off pins to build functionality.
)

(event Ability|EventOnEndAbility
  (params (bWasCancelled :type Boolean))
  (|CallParentFunction bWasCancelled)
  (Ability|RemoveGameplayEffectFromOwnerWithGrantedTags "(GameplayTags=((TagName=\"Granted.Athena.Item.Ashton.Turbo.ThrowBlock\")))") ; Fail safe to ensure blocking tag is removed
)
